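# Niagara system for the EvilCowboy skin's Stage 2 on-body transition, attaching smoke, glows and light to character sockets through a shared socket reader.
# SpiralSmoke cycles through seven sockets with drifting flipbook puffs, while Decal bursts nine sprites across nine other sockets.
# LightRays uses spawn groups to split stretched sprites between two sockets and adds light rendering.
# Variant selectors hold purple-toned default and blue alternate palettes; a time-of-day scalar scales alpha.
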NiagaraSystem /BRCosmetics/Athena/Effects/Particles/Skins/EvilCowboy/NS_Character_EvilCowboy_Stage02_OnBodyTransition
dependencies {
 usedDataInterfaces [
  &NiagaraDataInterfaceCurve
  &NiagaraDataInterfaceVectorCurve
  &NiagaraDataInterfaceSpriteRendererInfo
  &NiagaraDataInterfaceVector2DCurve
  &NiagaraDataInterfaceArrayFloat3
  &NiagaraDataInterfaceSocketReader
 ]
 usedDynamicInputs [&FNTimeOfDayScalar]
 usedModules [
  &self:SystemSpawnScript.SystemScriptSource.SystemScriptGraph.NiagaraNodeAssignment_0.SetVariables_#1
  &self:SystemSpawnScript.SystemScriptSource.SystemScriptGraph.NiagaraNodeAssignment_1.SetVariables_#2
  &SystemState
  &FN_VariantManager
 ]
}
emitters {
 Decal {
  emitterUpdateScript [
   &EmitterState(Life Cycle Mode=::NewEnumerator1, Loop Behavior=::NewEnumerator1, Loop Duration=1, UseLoopDelay=true)
   &SpawnBurst_Instantaneous(Spawn Count=9, Use Spawn Probability=false, UseLoopCountLimit=false)
  ]
  particleSpawnScript [
   /SetVariables(Particles.SkeletonSamplingIndex:NiagaraInt32=/CascadeToNiagaraConverter/NiagaraScripts/DynamicInputScripts/SequentialInt(Use Array Data:NiagaraBool=false, First Number:NiagaraInt32=7, Sequence:NiagaraDataInterfaceArrayInt32=?, Last Number:NiagaraInt32=15)) {isSetParameters=true}
   &InitializeParticle(
    Color=&VariantColorSelector(
     Variant Amount=0
     Original Color=&RandomRangeLinearColor(Minimum=<rgba 0.1175 0.1175 0.1175 2>, Maximum=<rgba 1 1 1 2>)
     Variant 1 Color=&RandomRangeLinearColor(Minimum=<rgba 0 0.667 1 3>, Maximum=<rgba 0 0.333 0.667 3>)
     Variant 2 Color=?
    )
    Sprite Rotation Mode=::NewEnumerator2
    Sprite Size Mode=::NewEnumerator3
    Uniform Sprite Size=50
    UsePositionOffset=false
    Write Material Random=true
   )
  ]
  particleUpdateScript [
   &ParticleState()
   &SocketLocation(
    SamplePosOffset=true
    Sampled Position Offset=&SelectVectorFromArray(
     Direct Array Index=$Particles.SkeletonSamplingIndex
     Vector Selection Array=DataInterface() {propertyValues {FloatData [<5 0 0>, <5 0 0>, <0 0 0>, <0 0 0>, <-10 0 0>, <10 0 0>, <-10 0 0>, <0 0 0>, <0 0 0>]}}
    )
    Socket Index=$Particles.SkeletonSamplingIndex
    Socket Reader=$System.Socket Reader
    Socket Sampling Mode=::NewEnumerator3
    VelocitySpeedLimit=false
   )
   &ScaleColor(
    Scale Alpha=&FloatFromCurve(
     FloatCurve=DataInterface() {propertyValues {Curve {keys [
         {arriveTangent=-0.1447, interpMode=RCIM_Cubic, leaveTangent=-0.1447, tangentMode=RCTM_User, time=0, value=0}
         {arriveTangent=-0.0772, interpMode=RCIM_Cubic, leaveTangent=-0.0772, tangentMode=RCTM_User, time=0.0919, value=0.2487}
         <at 1 0>
        ]}}}
     Scale Curve=$System.TODScalar
    )
    Scale RGB=/BRCosmetics/Weapons/FORT_Melee/Pickaxe_GateHound/FX/VariantManager/VariantVectorSelector(
     Original Vector:Vector3f=&VectorFromCurve(VectorCurve=DataInterface() {propertyValues {
        XCurve {keys [
          <at 0.25 0.8461 RCIM_Cubic>
          {arriveTangent=-0.4521, interpMode=RCIM_Cubic, leaveTangent=-0.4521, time=0.5, value=0.3291}
          {arriveTangent=-0.6583, interpMode=RCIM_Cubic, leaveTangent=-0.6583, time=0.7, value=0.6427}
          <at 1 0 RCIM_Cubic>
         ]}
        YCurve {keys [
          <at 0.25 0.1038 RCIM_Cubic>
          {arriveTangent=0.1823, interpMode=RCIM_Cubic, leaveTangent=0.1823, time=0.5, value=0.0779}
          {arriveTangent=-0.1559, interpMode=RCIM_Cubic, leaveTangent=-0.1559, time=0.7, value=0.1858}
          <at 1 0 RCIM_Cubic>
         ]}
        ZCurve {keys [
          <at 0.25 1 RCIM_Cubic>
          <at 0.5 0.3906 RCIM_Cubic>
          {arriveTangent=-0.7812, interpMode=RCIM_Cubic, leaveTangent=-0.7812, time=0.7, value=1}
          <at 1 0 RCIM_Cubic>
         ]}
       }})
     Variant 1 Vector:Vector3f=&VectorFromCurve(VectorCurve=DataInterface() {propertyValues {
        XCurve {keys [
          <at 0.25 0.1038 RCIM_Cubic>
          {arriveTangent=0.1819, interpMode=RCIM_Cubic, leaveTangent=0.1819, time=0.5, value=0.0779}
          {arriveTangent=-0.1559, interpMode=RCIM_Cubic, leaveTangent=-0.1559, time=0.701, value=0.1858}
          <at 1 0 RCIM_Cubic>
         ]}
        YCurve {keys [
          <at 0.25 0.7013 RCIM_Cubic>
          {arriveTangent=-0.9984, interpMode=RCIM_Cubic, leaveTangent=-0.9984, time=0.5, value=0.2864}
          {arriveTangent=-0.5728, interpMode=RCIM_Cubic, leaveTangent=-0.5728, time=0.701, value=0.251}
          <at 1 0 RCIM_Cubic>
         ]}
        ZCurve {keys [
          <at 0.25 1 RCIM_Cubic>
          {arriveTangent=-1.5886, interpMode=RCIM_Cubic, leaveTangent=-1.5886, time=0.5, value=0.3906}
          {arriveTangent=-0.7812, interpMode=RCIM_Cubic, leaveTangent=-0.7812, time=0.701, value=0.2836}
          <at 1 0 RCIM_Cubic>
         ]}
       }})
     Variant 2 Vector:Vector3f=?
     Variant 3 Vector:Vector3f=?
     Variant Amount:NiagaraInt32=0
    )
    ScaleA=true
    ScaleRGB=true
   )
   &CameraOffset(Camera Offset Amount=20)
   &DynamicMaterialParameters(
    Index 0 Param 1=&FloatFromCurve(FloatCurve=DataInterface() {propertyValues {Curve {keys [
         <at 0 -0.5 RCIM_Cubic>
         {arriveTangent=18.2487, interpMode=RCIM_Cubic, leaveTangent=18.2487, tangentMode=RCTM_User, time=0.0692, value=0.2738}
         {arriveTangent=4.8641, interpMode=RCIM_Cubic, leaveTangent=4.8641, tangentMode=RCTM_User, time=0.15, value=1}
         {arriveTangent=4.2839, interpMode=RCIM_Cubic, leaveTangent=4.2839, tangentMode=RCTM_User, time=1, value=5}
        ]}}})
    Index 0 Param 3=&FloatFromCurve(FloatCurve=DataInterface() {propertyValues {Curve {keys [
         {arriveTangent=0.1967, interpMode=RCIM_Cubic, leaveTangent=0.1967, tangentMode=RCTM_User, time=0, value=-0.25}
         {arriveTangent=0.1844, interpMode=RCIM_Cubic, leaveTangent=0.1844, tangentMode=RCTM_User, time=1, value=1}
        ]}}})
    Index 0 Param 4=1
    Module.Index0Float0Write=true
    Module.Index0Float1Write=true
    Module.Index0Float2Write=true
    Module.Index0Float3Write=true
   )
  ]
  rendererClasses [&NiagaraSpriteRendererProperties]
  renderers [{rendererClass=&NiagaraSpriteRendererProperties}]
 }
 LightRays {
  emitterUpdateScript [
   &EmitterState()
   &SpawnRate(SpawnRate=120, Use Spawn Probability=false)
   &SpawnBurst_Instantaneous(Spawn Group=1, Use Spawn Probability=false, UseLoopCountLimit=false)
  ]
  particleSpawnScript [
   &InitializeParticle(
    Color=&VariantColorSelector(Variant Amount=0, Original Color=<rgba 2 2 2 0.35>, Variant 1 Color=<rgba 2 2 2 0.275>, Variant 2 Color=?)
    Lifetime=&MergeHelper_FloatSwitch(
     0=&RandomRangeFloat(Minimum=0.25, Maximum=0.3, Evaluation Type=::NewEnumerator1)
     1=0.25
     2=?
     3=?
     Index=$Emitter.SpawnGroup
     Num=::NewEnumerator0
    )
    Sprite Rotation Mode=::NewEnumerator2
    Sprite Size=&RandomRangeVector2D(Minimum=<50 100>, Maximum=<100 250>, Evaluation Type=::NewEnumerator1)
    Sprite Size Mode=::NewEnumerator4
    Sprite UV Mode=::NewEnumerator1
    UsePositionOffset=false
    Write Material Random=true
   )
   &SocketLocation(
    SamplePosOffset=true
    Sampled Position Offset=&MergeHelper_VectorSwitch(0=<0 0 0>, 1=<0 0 10>, 2=?, 3=?, Index=$Emitter.SpawnGroup, Num=::NewEnumerator0)
    Sampled Position Offset Coordinate Space:ENiagaraSkelMeshTransforms=::NewEnumerator2
    Socket Index=&MergeHelper_IntSwitch(0=6, 1=16, 2=?, 3=?, 4=?, 5=?, Index=$Emitter.SpawnGroup, Num=::NewEnumerator0)
    Socket Reader=$System.Socket Reader
    Socket Sampling Mode=::NewEnumerator3
    VelocitySpeedLimit=false
   )
   &ShapeLocation(MaskSpawnGroup=false, Override Seed=false, Sphere Radius=2, UseOwnerScale=true)
   &SubUVAnimation(
    End Frame Range Override=3
    Evaluation Type=::NewEnumerator0
    "Sprite Renderer "=DataInterface() {propertyValues {SpriteRenderer=&self:LightRays_1.NiagaraSpriteRendererProperties_1}}
    SubUV Animation Mode:ENiagara_SubUVLookupModeV2=::NewEnumerator2
    UseEndFrame=true
    UseStartFrame=true
   )
   &AddVelocity(Velocity=&MergeHelper_VectorSwitch(0=<0 0 0>, 1=<0 0 220>, 2=?, 3=?, Index=$Emitter.SpawnGroup, Num=::NewEnumerator0))
   /SetVariables(Particles.VisibilityTag=$Emitter.SpawnGroup) {isSetParameters=true}
  ]
  particleUpdateScript [
   &ParticleState()
   &ScaleSpriteSize(
    Scale Factor=&Vector2DFromCurve(
     Vector2Curve=DataInterface() {propertyValues {
       XCurve {keys [
         {arriveTangent=4.9957, interpMode=RCIM_Cubic, leaveTangent=4.9957, tangentMode=RCTM_User, time=-0.0026, value=1}
         {arriveTangent=0.9974, interpMode=RCIM_Cubic, leaveTangent=0.9974, time=0.3904, value=1.8714}
         <at 1 2 RCIM_Cubic>
        ]}
       YCurve {keys [
         {arriveTangent=0.3115, interpMode=RCIM_Cubic, leaveTangent=0.3115, tangentMode=RCTM_User, time=-0.0026, value=0.8481}
         {arriveTangent=0.1515, interpMode=RCIM_Cubic, leaveTangent=0.1515, time=0.3904, value=0.9697}
         <at 1 1 RCIM_Cubic>
        ]}
      }}
     ScaleCurve=<1 1>
    )
    Scale Sprite Size Mode=::NewEnumerator2
   )
   &ScaleColor(
    Scale Alpha=&FloatFromCurve(
     FloatCurve=DataInterface() {propertyValues {Curve {keys [
         {arriveTangent=-0.1447, interpMode=RCIM_Cubic, leaveTangent=-0.1447, tangentMode=RCTM_User, time=0, value=0}
         {arriveTangent=-0.0772, interpMode=RCIM_Cubic, leaveTangent=-0.0772, tangentMode=RCTM_User, time=0.0186, value=1.0006}
         <at 1 0>
        ]}}}
     Scale Curve=$System.TODScalar
    )
    Scale RGB=&VariantVectorSelector(
     Original Vector:Vector3f=&VectorFromCurve(VectorCurve=DataInterface() {propertyValues {
        XCurve {keys [
          {arriveTangent=-0.139, time=0.25, value=0.3097}
          {arriveTangent=-0.1114, interpMode=RCIM_Cubic, leaveTangent=-0.1114, time=0.6, value=0.2777}
          {arriveTangent=-1.1392, leaveTangent=-1.1392, time=0.6738, value=0.2624}
          <at 1 0>
         ]}
        YCurve {keys [
          {arriveTangent=0.0234, time=0.25, value=0.0523}
          {arriveTangent=0.1447, interpMode=RCIM_Cubic, leaveTangent=0.1447, time=0.6, value=0.0868}
          {arriveTangent=-0.4345, leaveTangent=-0.4345, time=0.6738, value=0.1136}
          <at 1 0>
         ]}
        ZCurve {keys [
          {arriveTangent=0.0476, time=0.25, value=0.4167}
          {arriveTangent=1.3765, interpMode=RCIM_Cubic, leaveTangent=1.3765, time=0.6, value=0.3906}
          {arriveTangent=-2.0374, leaveTangent=-2.0374, time=0.6738, value=1}
          <at 1 0>
         ]}
       }})
     Variant 1 Vector:Vector3f=&VectorFromCurve(VectorCurve=DataInterface() {propertyValues {
        XCurve {keys [
          {arriveTangent=-0.139, time=0.25, value=0}
          {arriveTangent=0.2946, interpMode=RCIM_Cubic, leaveTangent=0.2946, time=0.6, value=0.2448}
          {arriveTangent=-1.1392, leaveTangent=-1.1392, time=0.6738, value=0.1248}
          <at 1 0>
         ]}
        YCurve {keys [
          {arriveTangent=0.0234, time=0.25, value=0.3401}
          {arriveTangent=-0.282, interpMode=RCIM_Cubic, leaveTangent=-0.282, time=0.6, value=0.5016}
          {arriveTangent=-0.4345, leaveTangent=-0.4345, time=0.6738, value=0.2205}
          <at 1 0>
         ]}
        ZCurve {keys [
          {arriveTangent=0.0476, time=0.25, value=1}
          {arriveTangent=-1.401, interpMode=RCIM_Cubic, leaveTangent=-1.401, time=0.6, value=1}
          {arriveTangent=-2.0374, leaveTangent=-2.0374, time=0.6738, value=0.4062}
          <at 1 0>
         ]}
       }})
     Variant 2 Vector:Vector3f=?
     Variant 3 Vector:Vector3f=?
     Variant Amount:NiagaraInt32=0
    )
    ScaleA=true
    ScaleRGB=true
   )
   &CameraOffset(Camera Offset Amount=5)
   &FN_Light_Attributes(
    Color=&MakeLinearColorFromVectorAndFloat(
     "Vector (RGB)"=&VariantVectorSelector(
      Original Vector:Vector3f=&VectorFromCurve(
       VectorCurve=DataInterface() {propertyValues {
         XCurve {keys [
           {arriveTangent=-0.139, time=0.25, value=0.3097}
           {arriveTangent=-0.1114, interpMode=RCIM_Cubic, leaveTangent=-0.1114, time=0.6, value=0.2777}
           {arriveTangent=-1.1392, leaveTangent=-1.1392, time=0.6738, value=0.2624}
           <at 1 0>
          ]}
         YCurve {keys [
           {arriveTangent=0.0234, time=0.25, value=0.0523}
           {arriveTangent=0.1447, interpMode=RCIM_Cubic, leaveTangent=0.1447, time=0.6, value=0.0868}
           {arriveTangent=-0.4345, leaveTangent=-0.4345, time=0.6738, value=0.1136}
           <at 1 0>
          ]}
         ZCurve {keys [
           {arriveTangent=0.0476, time=0.25, value=0.4167}
           {arriveTangent=1.3765, interpMode=RCIM_Cubic, leaveTangent=1.3765, time=0.6, value=0.3906}
           {arriveTangent=-2.0374, leaveTangent=-2.0374, time=0.6738, value=1}
           <at 1 0>
          ]}
        }}
       Scale Curve=<15 15 15>
      )
      Variant 1 Vector:Vector3f=&VectorFromCurve(
       VectorCurve=DataInterface() {propertyValues {
         XCurve {keys [{arriveTangent=-0.139, time=0, value=0.1562}, <at 1 0>]}
         YCurve {keys [{arriveTangent=0.0234, time=0, value=0.4432}, <at 1 0>]}
         ZCurve {keys [{arriveTangent=0.0476, time=0, value=1}, <at 1 0>]}
        }}
       Scale Curve=<0 7.5 15>
      )
      Variant 2 Vector:Vector3f=?
      Variant 3 Vector:Vector3f=?
      Variant Amount:NiagaraInt32=0
     )
     "Float (Alpha)"=&FloatFromCurve(
      FloatCurve=DataInterface() {propertyValues {Curve {keys [{arriveTangent=-2.1518, interpMode=RCIM_Cubic, leaveTangent=-2.1518, tangentMode=RCTM_User, time=0, value=1}, <at 1 0>]}}}
      Scale Curve=8
     )
    )
    Create Color Attribute=true
    Create Exponent Attribute=true
    Create Radius Attribute=true
    Exponent=5
    Radius=100
   )
   &SolveForcesAndVelocity(Clamp Velocity=false, Limit Acceleration=false)
  ]
  rendererClasses [&NiagaraSpriteRendererProperties, &NiagaraLightRendererProperties]
  renderers [
   {rendererClass=&NiagaraSpriteRendererProperties}
   {rendererClass=&NiagaraLightRendererProperties, rendererIndex=1}
  ]
 }
 SpiralSmoke {
  emitterUpdateScript [
   &EmitterState()
   &SpawnRate(SpawnRate=100, Use Spawn Probability=false)
  ]
  particleSpawnScript [
   &InitializeParticle(
    Color=&VariantColorSelector(
     Variant Amount=0
     Original Color=&RandomRangeLinearColor(Minimum=<rgba 0.4649 0 0.5 1>, Maximum=<rgba 0.9896 0.099 0.886 2>)
     Variant 1 Color=&RandomRangeLinearColor(Minimum=<rgba 0 0.5 1 2>, Maximum=<rgba 0 0.5 1 1.5>)
     Variant 2 Color=?
    )
    Lifetime=&RandomRangeFloat(Minimum=0.95, Maximum=1.33)
    Sprite Rotation Mode=::NewEnumerator2
    Sprite Size Mode=::NewEnumerator3
    Sprite UV Mode=::NewEnumerator3
    Uniform Sprite Size=&RandomRangeFloat(Minimum=12, Maximum=15)
    UsePositionOffset=false
    Write Material Random=true
   )
   &SocketLocation(
    SamplePosOffset=true
    Sampled Position Offset=&MakeVector(X=0, Y=0, Z=&RandomRangeFloat(Minimum=0, Maximum=25))
    Sampled Position Offset Coordinate Space:ENiagaraSkelMeshTransforms=::NewEnumerator2
    Socket Index=&SequentialInt(Use Array Data:NiagaraBool=false, First Number:NiagaraInt32=0, Sequence:NiagaraDataInterfaceArrayInt32=?, Last Number:NiagaraInt32=6)
    Socket Reader=$System.Socket Reader
    Socket Sampling Mode=::NewEnumerator3
    VelocitySpeedLimit=false
   )
   &ShapeLocation(MaskSpawnGroup=false, Override Seed=false, Sphere Radius=2, UseOwnerScale=true)
   &AddVelocity(
    Module.Use Falloff Distance=false
    Override Seed=false
    Velocity Mode=::NewEnumerator1
    Velocity Origin=$Output.SocketLocation.SampledPosition
    Velocity Speed=2
   ) {module=AddVelocity001}
   &AddVelocity(Velocity=&MakeVector(X=0, Y=0, Z=&RandomRangeFloat(Minimum=25, Maximum=80)))
   &DynamicMaterialParameters(
    Index 0 Param 1=&RandomRangeFloat(Minimum=0, Maximum=4, Evaluation Type=::NewEnumerator1)
    Index 0 Param 2=&RandomRangeFloat(Minimum=1, Maximum=4, Evaluation Type=::NewEnumerator1)
    Index 0 Param 3=125
    Index 0 Param 4=0.5
    Module.Index0Float0Write=true
    Module.Index0Float1Write=true
    Module.Index0Float2Write=true
    Module.Index0Float3Write=true
   )
  ]
  particleUpdateScript [
   &ParticleState()
   &Drag(Drag=2, Use Linear Drag=true, Use Rotational Drag=true)
   &ScaleSpriteSize(
    Scale Sprite Size Mode=::NewEnumerator0
    Uniform Scale Factor=&FloatFromCurve(FloatCurve=DataInterface() {propertyValues {Curve {keys [
         {arriveTangent=16.9657, interpMode=RCIM_Cubic, leaveTangent=16.9657, tangentMode=RCTM_User, time=0, value=1}
         {arriveTangent=4.2114, interpMode=RCIM_Cubic, leaveTangent=4.2114, time=0.2637, value=3.9734}
         <at 1.0038 5.2275 RCIM_Cubic>
        ]}}})
   )
   &ScaleColor(
    Scale Alpha=/BRCosmetics/Weapons/FORT_Melee/Pickaxe_GateHound/FX/VariantManager/VariantFloatSelector(
     Original Float:NiagaraFloat=&FloatFromCurve(
      FloatCurve=DataInterface() {propertyValues {Curve {keys [
          <at 0 0 RCIM_Cubic>
          {arriveTangent=3.8807, interpMode=RCIM_Cubic, leaveTangent=3.8807, tangentMode=RCTM_User, time=0.0422, value=1.2316}
          {arriveTangent=-4.6011, interpMode=RCIM_Cubic, leaveTangent=-4.6011, tangentMode=RCTM_User, time=0.075, value=1.3402}
          <at 1 0 RCIM_Cubic>
         ]}}}
      Scale Curve=$System.TODScalar
     )
     Variant 1 Float:NiagaraFloat=&FloatFromCurve(
      FloatCurve=DataInterface() {propertyValues {Curve {keys [
          <at 0 0 RCIM_Cubic>
          {arriveTangent=3.8807, interpMode=RCIM_Cubic, leaveTangent=3.8807, tangentMode=RCTM_User, time=0.0422, value=1.2316}
          {arriveTangent=-4.6011, interpMode=RCIM_Cubic, leaveTangent=-4.6011, tangentMode=RCTM_User, time=0.075, value=1.3402}
          <at 1 0 RCIM_Cubic>
         ]}}}
      Scale Curve=$System.TODScalar
     )
     Variant 2 Float:NiagaraFloat=?
     Variant 3 Float:NiagaraFloat=?
     Variant Amount:NiagaraInt32=0
    )
    Scale RGB=&VariantVectorSelector(
     Original Vector:Vector3f=&VectorFromCurve(VectorCurve=DataInterface() {propertyValues {
        XCurve {keys [<at 0 1 RCIM_Cubic>, {arriveTangent=-2.8571, interpMode=RCIM_Cubic, leaveTangent=-2.8571, time=0.15, value=0.8785}, <at 0.35 0 RCIM_Cubic>]}
        YCurve {keys [<at 0 1 RCIM_Cubic>, {arriveTangent=-2.8571, interpMode=RCIM_Cubic, leaveTangent=-2.8571, time=0.15, value=0.1156}, <at 0.35 0 RCIM_Cubic>]}
        ZCurve {keys [<at 0 1 RCIM_Cubic>, {arriveTangent=-2.8571, interpMode=RCIM_Cubic, leaveTangent=-2.8571, time=0.15, value=2}, <at 0.35 0 RCIM_Cubic>]}
       }})
     Variant 1 Vector:Vector3f=&VectorFromCurve(VectorCurve=DataInterface() {propertyValues {
        XCurve {keys [<at 0 0.5354 RCIM_Cubic>, {arriveTangent=-0.6693, interpMode=RCIM_Cubic, leaveTangent=-0.6693, time=0.15, value=0.5}, <at 0.8 0 RCIM_Cubic>]}
        YCurve {keys [
          <at 0 0.6727 RCIM_Cubic>
          {arriveTangent=-0.8409, interpMode=RCIM_Cubic, leaveTangent=-0.8409, time=0.15, value=0.6667}
          <at 0.8 0 RCIM_Cubic>
         ]}
        ZCurve {keys [<at 0 1 RCIM_Cubic>, {arriveTangent=-1.25, interpMode=RCIM_Cubic, leaveTangent=-1.25, time=0.15, value=1}, <at 0.8 0 RCIM_Cubic>]}
       }})
     Variant 2 Vector:Vector3f=?
     Variant 3 Vector:Vector3f=?
     Variant Amount:NiagaraInt32=0
    )
    ScaleA=true
    ScaleRGB=true
   )
   &AccelerationForce(Acceleration=&MakeVector(X=0, Y=&RandomRangeFloat(Minimum=-25, Maximum=-35), Z=2))
   &SolveForcesAndVelocity(Clamp Velocity=false, Limit Acceleration=false)
   &SubUVAnimation(
    End Frame Range Override=62
    "Sprite Renderer "=DataInterface() {propertyValues {SpriteRenderer=&self:SpiralSmoke_0.NiagaraSpriteRendererProperties_1}}
    Start Frame Range Override=30
    UseEndFrame=true
    UseStartFrame=true
   )
   &CameraOffset(Camera Offset Amount=&FloatFromCurve(FloatCurve=DataInterface() {propertyValues {Curve {keys [<at 0 25>,<at 1 50>]}}}))
  ]
  rendererClasses [&NiagaraSpriteRendererProperties]
  renderers [{rendererClass=&NiagaraSpriteRendererProperties}]
 }
}
summary {
 systemName=NS_Character_EvilCowboy_Stage02_OnBodyTransition
 emitters [
  {emitterName=SpiralSmoke, rendererClasses [&NiagaraSpriteRendererProperties]}
  {emitterName=LightRays, rendererClasses [&NiagaraSpriteRendererProperties, &NiagaraLightRendererProperties]}
  {emitterName=Decal, rendererClasses [&NiagaraSpriteRendererProperties]}
 ]
 userVariables [User.Variant {type=NiagaraFloat, defaultValue=0, description=""}]
}
