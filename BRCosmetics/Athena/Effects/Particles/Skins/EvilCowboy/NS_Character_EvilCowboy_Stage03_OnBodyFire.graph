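# Niagara system for looping on-body fire on the Stage 3 EvilCowboy character skin, with green and blue color variants.
# Mesh flames, flipbook sprite flames, decal-style sprites and two lights follow sockets through a shared Socket Reader.
# Spawn groups and partitions select sockets per particle, with spawn group copied to VisibilityTag to switch curves in DecalLoop.
# A time-of-day scalar scales flame and decal alpha plus light color.
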NiagaraSystem /BRCosmetics/Athena/Effects/Particles/Skins/EvilCowboy/NS_Character_EvilCowboy_Stage03_OnBodyFire
dependencies {
 usedDataInterfaces [
  &NiagaraDataInterfaceCurve
  &NiagaraDataInterfaceArrayFloat3
  &NiagaraDataInterfaceSpriteRendererInfo
  &NiagaraDataInterfaceVectorCurve
  &NiagaraDataInterfaceSocketReader
 ]
 usedDynamicInputs [&MakeFloatFromVector2d, &FNTimeOfDayScalar]
 usedModules [
  &self:SystemSpawnScript.SystemScriptSource.SystemScriptGraph.NiagaraNodeAssignment_0.SetVariables_#1
  &self:SystemSpawnScript.SystemScriptSource.SystemScriptGraph.NiagaraNodeAssignment_1.SetVariables_#2
  &SystemState
  &FN_VariantManager
 ]
}
emitters {
 DecalLoop {
  emitterUpdateScript [
   &EmitterState(Inactive Response=::NewEnumerator1, Life Cycle Mode=::NewEnumerator1, Loop Duration=1, UseLoopDelay=false)
   &SpawnRate(SpawnRate=5, Use Spawn Probability=false)
   &SpawnRate(SpawnGroup=1, SpawnRate=4, Use Spawn Probability=false) {module=SpawnRate001}
  ]
  particleSpawnScript [
   &InitializeParticle(
    Color=&VariantColorSelector(
     Variant Amount=0
     Original Color=&RandomRangeLinearColor(Minimum=<rgba 1 2 0 0.1>, Maximum=<rgba 1 2 0 0.165>)
     Variant 1 Color=&RandomRangeLinearColor(Minimum=<rgba 0 0.33 1 0.3>, Maximum=<rgba 0 0.5 1 0.5>)
     Variant 2 Color=?
    )
    Lifetime=&MergeHelper_FloatSwitch(0=2, 1=1.5, 2=?, 3=?, Index=$Emitter.SpawnGroup, Num=::NewEnumerator0)
    Sprite Size Mode=::NewEnumerator3
    Sprite UV Mode=::NewEnumerator1
    Uniform Sprite Size=&MergeHelper_FloatSwitch(
     0=&RandomRangeFloat(Minimum=30, Maximum=40, Evaluation Type=::NewEnumerator1)
     1=&RandomRangeFloat(Minimum=10, Maximum=15, Evaluation Type=::NewEnumerator1)
     2=&RandomRangeFloat(Minimum=10, Maximum=15, Evaluation Type=::NewEnumerator1)
     3=?
     Index=$Emitter.SpawnGroup
     Num=::NewEnumerator0
    )
    UsePositionOffset=false
    Write Material Random=true
   )
   /SetVariables(Particles.VisibilityTag=$Emitter.SpawnGroup) {isSetParameters=true}
   &ShapeLocation(MaskSpawnGroup=false, Override Seed=false, Sphere Radius=2, UseOwnerScale=true)
  ]
  particleUpdateScript [
   &ParticleState()
   &ScaleSpriteSize(
    Scale Factor=&MergeHelper_Vector2Switch(
     0=&Vector2DFromFloat(Value=&FloatFromCurve(FloatCurve=DataInterface() {propertyValues {Curve {keys [<at 0 1>,<at 1 0>]}}}))
     1=<1 1>
     2=?
     3=?
     4=?
     Index=$Particles.VisibilityTag
     Num=::NewEnumerator0
    )
    Scale Sprite Size Mode=::NewEnumerator2
   )
   &ScaleColor(
    Scale Alpha=&MergeHelper_FloatSwitch(
     0=&FloatFromCurve(
      FloatCurve=DataInterface() {propertyValues {Curve {keys [<at 0 0 RCIM_Cubic>,<at 0.4 1 RCIM_Cubic>,<at 1 0 RCIM_Cubic>]}}}
      Scale Curve=$System.TODScalar
     )
     1=&FloatFromCurve(FloatCurve=DataInterface() {propertyValues {Curve {keys [<at 0 0>,<at 0.25 0.25>,<at 0.6 0.1>,<at 1 0>]}}}, Scale Curve=$System.TODScalar)
     2=?
     3=?
     Index=$Particles.VisibilityTag
     Num=::NewEnumerator0
    )
    Scale RGB=&MergeHelper_VectorSwitch(
     0=&VectorFromCurve(VectorCurve=DataInterface() {propertyValues {
        XCurve {keys [<at 0 1 RCIM_Cubic>,<at 0.3 1.25>,<at 1 0>]}
        YCurve {keys [<at 0 1 RCIM_Cubic>,<at 0.3 1.25>,<at 1 0>]}
        ZCurve {keys [<at 0 1 RCIM_Cubic>,<at 0.3 1.25>,<at 1 0>]}
       }})
     1=&VectorFromCurve(VectorCurve=DataInterface() {propertyValues {
        XCurve {keys [<at 0 0 RCIM_Cubic>,<at 0.5 1.25>]}
        YCurve {keys [<at 0 0 RCIM_Cubic>,<at 0.5 1.25>]}
        ZCurve {keys [<at 0 0 RCIM_Cubic>,<at 0.5 1.25>]}
       }})
     2=?
     3=?
     Index=$Particles.VisibilityTag
     Num=::NewEnumerator0
    )
    ScaleA=true
    ScaleRGB=true
   )
   &CameraOffset(Camera Offset Amount=10)
   &SocketLocation(
    Orientation Sampling=::NewEnumerator1
    SamplePosOffset=true
    Socket Index=&MergeHelper_IntSwitch(
     0=/CascadeToNiagaraConverter/NiagaraScripts/DynamicInputScripts/SequentialInt(Use Array Data:NiagaraBool=false, First Number:NiagaraInt32=12, Sequence:NiagaraDataInterfaceArrayInt32=?, Last Number:NiagaraInt32=13)
     1=&SequentialInt(Use Array Data:NiagaraBool=false, First Number:NiagaraInt32=10, Sequence:NiagaraDataInterfaceArrayInt32=?, Last Number:NiagaraInt32=11)
     2=&SequentialInt(Use Array Data:NiagaraBool=false, First Number:NiagaraInt32=10, Sequence:NiagaraDataInterfaceArrayInt32=?, Last Number:NiagaraInt32=11)
     3=?
     4=?
     5=?
     Index=$Particles.VisibilityTag
     Num=::NewEnumerator0
    )
    Socket Reader=$System.Socket Reader
    Socket Sampling Mode=::NewEnumerator3
    VelocitySpeedLimit=false
   )
   &DynamicMaterialParameters(
    Index 0 Param 1=&FloatFromCurve(FloatCurve=DataInterface() {propertyValues {Curve {keys [<at 0 0 RCIM_Cubic>, {arriveTangent=5, interpMode=RCIM_Cubic, leaveTangent=5, time=0.4, value=0}, <at 1 5 RCIM_Cubic>]}}})
    Index 0 Param 2=1
    Index 0 Param 3=&FloatFromCurve(FloatCurve=DataInterface() {propertyValues {Curve {keys [<at 0 0>,<at 1 4>]}}})
    Index 0 Param 4=&RandomRangeFloat(Minimum=0, Maximum=1)
    Module.Index0Float0Write=true
    Module.Index0Float1Write=true
    Module.Index0Float2Write=true
    Module.Index0Float3Write=true
   )
  ]
  rendererClasses [&NiagaraSpriteRendererProperties]
  renderers [
   {rendererClass=&NiagaraSpriteRendererProperties}
   {rendererClass=&NiagaraSpriteRendererProperties, rendererIndex=1}
  ]
 }
 Flames_Loop {
  emitterUpdateScript [
   &EmitterState(Inactive Response=::NewEnumerator1, Life Cycle Mode=::NewEnumerator1, Loop Duration=1, UseLoopDelay=false)
   &SpawnBurst_Instantaneous(Use Spawn Probability=false, UseLoopCountLimit=false)
   &SpawnRate(SpawnRate=10, Use Spawn Probability=false)
   &SpawnBurst_Instantaneous(Spawn Group=1, Use Spawn Probability=false, UseLoopCountLimit=false) {module=SpawnBurst_Instantaneous001}
   &SpawnRate(SpawnGroup=1, SpawnRate=10, Use Spawn Probability=false) {module=SpawnRate001}
   /SetVariables(
    Emitter.ColorHigh:LinearColor=&VariantColorSelector(Variant Amount=0, Original Color=<rgba 0.5987 1 0.038 1>, Variant 1 Color=<rgba 0 0.4577 1 1>, Variant 2 Color=?)
    Emitter.ColorLow:LinearColor=&VariantColorSelector(Variant Amount=0, Original Color=<rgba 0 0.5052 0.0569 1>, Variant 1 Color=<rgba 0 0.2994 0.5052 1>, Variant 2 Color=?)
   ) {isSetParameters=true}
  ]
  particleSpawnScript [
   &PartitionParticles(Number Of Partitions=7)
   &InitializeParticle(
    Color=&VariantColorSelector(Variant Amount=0, Original Color=<rgba 1 1 1 0.085>, Variant 1 Color=<rgba 1 1 1 0.35>, Variant 2 Color=?)
    Lifetime=&RandomRangeFloat(Minimum=0.7, Maximum=0.9, Evaluation Type=::NewEnumerator1)
    Mesh Scale Max=<1 1 1>
    Mesh Scale Min=<0.8 0.8 0.5>
    Mesh Scale Mode=::NewEnumerator2
    UsePositionOffset=false
    Write Material Random=true
   )
   &DynamicMaterialParameters(
    Index 0 Param 1=&RandomRangeFloat(Minimum=0, Maximum=1, Evaluation Type=::NewEnumerator1)
    Module.Index0Float0Write=true
    Module.Index0Float1Write=false
    Module.Index0Float2Write=false
    Module.Index0Float3Write=false
   )
   /SetVariables(Particles.VisibilityTag=$Emitter.SpawnGroup) {isSetParameters=true}
  ]
  particleUpdateScript [
   &ParticleState()
   &SocketLocation(
    Orientation Sampling=::NewEnumerator1
    SamplePosOffset=true
    Socket Index=&MergeHelper_IntSwitch(
     0=&MergeHelper_IntSwitch(0=10, 1=0, 2=1, 3=2, 4=3, 5=4, Index=$Particles.PartitionParticles.Partition, Num=::NewEnumerator4)
     1=&MergeHelper_IntSwitch(0=11, 1=5, 2=6, 3=7, 4=8, 5=9, Index=$Particles.PartitionParticles.Partition, Num=::NewEnumerator4)
     2=?
     3=?
     4=?
     5=?
     Index=$Particles.VisibilityTag
     Num=::NewEnumerator0
    )
    Socket Reader=$System.Socket Reader
    Socket Sampling Mode=::NewEnumerator3
    VelocitySpeedLimit=false
   )
   &ScaleColor(
    Scale Alpha=&FloatFromCurve(FloatCurve=DataInterface() {propertyValues {Curve {keys [<at 0 0>,<at 0.25 1>,<at 0.35 1>,<at 1 0>]}}}, Scale Curve=$System.TODScalar)
    ScaleA=true
    ScaleRGB=true
   )
  ]
  rendererClasses [&NiagaraMeshRendererProperties]
  renderers [
   {rendererClass=&NiagaraMeshRendererProperties, rendererIndex=0}
   {rendererClass=&NiagaraMeshRendererProperties, rendererIndex=1}
  ]
 }
 Flames_Random {
  emitterUpdateScript [
   &EmitterState(Inactive Response=::NewEnumerator1, Life Cycle Mode=::NewEnumerator1, Loop Duration=1, UseLoopDelay=false)
   &SpawnRate(Spawn Probability=0.75, SpawnRate=15, Use Spawn Probability=true)
  ]
  particleSpawnScript [
   &PartitionParticles()
   /SetVariables(Particles.SkeletonSamplingIndex:NiagaraInt32=&MergeHelper_IntSwitch(
     0=&SequentialInt(Use Array Data:NiagaraBool=false, First Number:NiagaraInt32=0, Sequence:NiagaraDataInterfaceArrayInt32=?, Last Number:NiagaraInt32=4)
     1=&SequentialInt(Use Array Data:NiagaraBool=false, First Number:NiagaraInt32=5, Sequence:NiagaraDataInterfaceArrayInt32=?, Last Number:NiagaraInt32=9)
     2=?
     3=?
     4=?
     5=?
     Index=$Particles.PartitionParticles.Partition
     Num=::NewEnumerator0
    )) {isSetParameters=true}
   &InitializeParticle(
    Color=&VariantColorSelector(
     Variant Amount=0
     Original Color=&RandomRangeLinearColor(Minimum=<rgba 0.063 1 0.3466 0.2>, Maximum=<rgba 0.3053 5 0.594 0.225>)
     Variant 1 Color=&RandomRangeLinearColor(Minimum=<rgba 0 0.33 1 0.35>, Maximum=<rgba 0 2.5 5 0.3>)
     Variant 2 Color=?
    )
    Lifetime=&RandomRangeFloat(Minimum=0.4, Maximum=0.6, Evaluation Type=::NewEnumerator1)
    Sprite Rotation Mode=::NewEnumerator2
    Sprite Size=&Vector2DFromFloat(Value=&RandomRangeFloat(Minimum=10, Maximum=15, Evaluation Type=::NewEnumerator1))
    Sprite Size Mode=::NewEnumerator4
    UsePositionOffset=false
    Write Material Random=true
   )
   &ShapeLocation(MaskSpawnGroup=false, Override Seed=false, Sphere Radius=5, UseOwnerScale=true)
  ]
  particleUpdateScript [
   &ParticleState()
   &ScaleSpriteSize(
    Scale Sprite Size Mode=::NewEnumerator0
    Uniform Scale Factor=&FloatFromCurve(FloatCurve=DataInterface() {propertyValues {Curve {keys [
         {arriveTangent=11.6538, interpMode=RCIM_Cubic, leaveTangent=11.6538, tangentMode=RCTM_User, time=0, value=0.0211}
         {arriveTangent=0.5535, interpMode=RCIM_Cubic, leaveTangent=0.5535, tangentMode=RCTM_User, time=0.1836, value=0.9874}
         <at 1.0052 0.8728 RCIM_Cubic>
        ]}}})
   )
   &SocketLocation(
    Orientation Sampling=::NewEnumerator1
    SamplePosOffset=true
    Sampled Position Offset=&MergeHelper_VectorSwitch(
     0=&SelectVectorFromArray(
      Direct Array Index=$Particles.SkeletonSamplingIndex
      Vector Selection Array=DataInterface() {propertyValues {FloatData [<0 0 -10>, <0 0 0>, <0 0 0>, <0 0 0>, <0 0 0>]}}
     )
     1=<0 0 0>
     2=?
     3=?
     Index=$Particles.PartitionParticles.Partition
     Num=::NewEnumerator0
    )
    Socket Index=$Particles.SkeletonSamplingIndex
    Socket Reader=$System.Socket Reader
    Socket Sampling Mode=::NewEnumerator3
    VelocitySpeedLimit=false
   )
   &SubUVAnimation(
    End Frame Range Override=62
    "Sprite Renderer "=DataInterface() {propertyValues {SpriteRenderer=&self:Flames_Random_0.NiagaraSpriteRendererProperties_1}}
    Start Frame Range Override=16
    UseEndFrame=true
    UseStartFrame=true
   )
   &ScaleColor(
    Scale Alpha=&FloatFromCurve(
     FloatCurve=DataInterface() {propertyValues {Curve {keys [
         {arriveTangent=11.7282, interpMode=RCIM_Cubic, leaveTangent=11.7282, tangentMode=RCTM_User, time=0, value=0}
         <at 0.2 1 RCIM_Cubic>
         <at 1 0>
        ]}}}
     Scale Curve=$System.TODScalar
    )
    Scale RGB=&VectorFromCurve(VectorCurve=DataInterface() {propertyValues {XCurve {keys [<at 0.1 0>,<at 0.5 1>]}, YCurve {keys [<at 0.1 0.1>,<at 0.5 0.8266>]}, ZCurve {keys [<at 0.1 0>,<at 0.5 0.0249>]}}})
    ScaleA=true
    ScaleRGB=true
   )
   &CameraOffset(Camera Offset Amount=15)
  ]
  rendererClasses [&NiagaraSpriteRendererProperties]
  renderers [{rendererClass=&NiagaraSpriteRendererProperties}]
 }
 Light {
  emitterUpdateScript [
   &EmitterState(
    Inactive Response=::NewEnumerator1
    Life Cycle Mode=::NewEnumerator1
    Loop Behavior=::NewEnumerator1
    Loop Duration Mode=::NewEnumerator1
    UseLoopDelay=false
   )
   &SpawnBurst_Instantaneous(Spawn Count=2, Use Spawn Probability=false, UseLoopCountLimit=false)
  ]
  particleSpawnScript [
   &PartitionParticles()
   &InitializeParticle(UsePositionOffset=false, Write Material Random=true)
  ]
  particleUpdateScript [
   &ParticleState(Kill Particles When Lifetime Has Elapsed=false)
   &SocketLocation(
    Orientation Sampling=::NewEnumerator1
    SamplePosOffset=true
    Sampled Position Offset=&SelectVectorFromArray(
     Direct Array Index=$Particles.PartitionParticles.Partition
     Vector Selection Array=DataInterface() {propertyValues {FloatData [<-15 -10 5>, <15 10 -5>]}}
    )
    Socket Index=&MergeHelper_IntSwitch(
     0=12
     1=13
     2=&SequentialInt(Use Array Data:NiagaraBool=false, First Number:NiagaraInt32=10, Sequence:NiagaraDataInterfaceArrayInt32=?, Last Number:NiagaraInt32=11)
     3=?
     4=?
     5=?
     Index=$Particles.PartitionParticles.Partition
     Num=::NewEnumerator0
    )
    Socket Reader=$System.Socket Reader
    Socket Sampling Mode=::NewEnumerator3
    VelocitySpeedLimit=false
   )
   &FN_Light_Attributes(
    Color=&VariantColorSelector(Variant Amount=0, Original Color=<rgba 5.4794 25 0.6236 1.5>, Variant 1 Color=<rgba 0 12.5 25 1.5>, Variant 2 Color=?)
    Create Color Attribute=true
    Create Exponent Attribute=true
    Create Radius Attribute=true
    Exponent=2
    Radius=24
    Scale Color=&FloatFromCurve(
     FloatCurve=DataInterface() {propertyValues {Curve {keys [<at 0 0 RCIM_Cubic>,<at 1 1 RCIM_Cubic>]}}}
     CurveIndex=$Emitter.Age
     Scale Curve=$System.TODScalar
    )
   )
  ]
  rendererClasses [&NiagaraLightRendererProperties]
  renderers [{rendererClass=&NiagaraLightRendererProperties, rendererIndex=0}]
 }
}
summary {
 systemName=NS_Character_EvilCowboy_Stage03_OnBodyFire
 emitters [
  {emitterName=Flames_Random, rendererClasses [&NiagaraSpriteRendererProperties]}
  {emitterName=Flames_Loop, rendererClasses [&NiagaraMeshRendererProperties]}
  {emitterName=DecalLoop, rendererClasses [&NiagaraSpriteRendererProperties]}
  {emitterName=Light, rendererClasses [&NiagaraLightRendererProperties]}
 ]
 userVariables [User.Variant {type=NiagaraFloat, defaultValue=0, description=""}]
}
