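# Niagara system for the EvilCowboy skin's Stage 03 on-body transition, layering decals, electric arcs, light rays and spiral smoke on the character.
# Emitters place particles at skeletal sockets through a shared Socket Reader, using sequential indices and per-socket offsets.
# Spawn-group switches let single emitters pack several looks, feeding sprite, mesh and light renderers.
# Variant selectors hold alternate palettes, and a time-of-day scalar scales opacity.
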NiagaraSystem /BRCosmetics/Athena/Effects/Particles/Skins/EvilCowboy/NS_Character_EvilCowboy_Stage03_OnBodyTransition
dependencies {
 usedDataInterfaces [
  &NiagaraDataInterfaceSpriteRendererInfo
  &NiagaraDataInterfaceVectorCurve
  &NiagaraDataInterfaceCurve
  &NiagaraDataInterfaceVector2DCurve
  &NiagaraDataInterfaceArrayFloat3
  &NiagaraDataInterfaceSocketReader
 ]
 usedDynamicInputs [&FNTimeOfDayScalar]
 usedModules [
  &self:SystemSpawnScript.SystemScriptSource.SystemScriptGraph.NiagaraNodeAssignment_0.SetVariables_#1
  &self:SystemSpawnScript.SystemScriptSource.SystemScriptGraph.NiagaraNodeAssignment_1.SetVariables_#2
  &SystemState
  &FN_VariantManager
 ]
}
emitters {
 Decal {
  emitterUpdateScript [
   &EmitterState(Life Cycle Mode=::NewEnumerator1, Loop Behavior=::NewEnumerator1, Loop Duration=1, UseLoopDelay=true)
   &SpawnBurst_Instantaneous(Spawn Count=9, Use Spawn Probability=false, UseLoopCountLimit=false)
  ]
  particleSpawnScript [
   /SetVariables(Particles.SkeletonSamplingIndex:NiagaraInt32=/CascadeToNiagaraConverter/NiagaraScripts/DynamicInputScripts/SequentialInt(Use Array Data:NiagaraBool=false, First Number:NiagaraInt32=7, Sequence:NiagaraDataInterfaceArrayInt32=?, Last Number:NiagaraInt32=15)) {isSetParameters=true}
   &InitializeParticle(
    Color=&VariantColorSelector(
     Variant Amount=0
     Original Color=&RandomRangeLinearColor(Minimum=<rgba 0.1175 0.1175 0.1175 2>, Maximum=<rgba 1 1 1 2>)
     Variant 1 Color=&RandomRangeLinearColor(Minimum=<rgba 0.1175 0.1175 0.1175 4>, Maximum=<rgba 1 1 1 4>)
     Variant 2 Color=?
    )
    Sprite Rotation Mode=::NewEnumerator2
    Sprite Size Mode=::NewEnumerator3
    Uniform Sprite Size=50
    UsePositionOffset=false
    Write Material Random=true
   )
  ]
  particleUpdateScript [
   &ParticleState()
   &SocketLocation(
    SamplePosOffset=true
    Sampled Position Offset=&SelectVectorFromArray(
     Direct Array Index=$Particles.SkeletonSamplingIndex
     Vector Selection Array=DataInterface() {propertyValues {FloatData [<5 0 0>, <5 0 0>, <0 0 0>, <0 0 0>, <-10 0 0>, <10 0 0>, <-10 0 0>, <0 0 0>, <0 0 0>]}}
    )
    Socket Index=$Particles.SkeletonSamplingIndex
    Socket Reader=$System.Socket Reader
    Socket Sampling Mode=::NewEnumerator3
    VelocitySpeedLimit=false
   )
   &ScaleColor(
    Scale Alpha=&FloatFromCurve(
     FloatCurve=DataInterface() {propertyValues {Curve {keys [
         {arriveTangent=-0.1447, interpMode=RCIM_Cubic, leaveTangent=-0.1447, tangentMode=RCTM_User, time=0, value=0}
         {arriveTangent=-0.0772, interpMode=RCIM_Cubic, leaveTangent=-0.0772, tangentMode=RCTM_User, time=0.0919, value=0.2487}
         <at 1 0>
        ]}}}
     Scale Curve=$System.TODScalar
    )
    Scale RGB=/BRCosmetics/Weapons/FORT_Melee/Pickaxe_GateHound/FX/VariantManager/VariantVectorSelector(
     Original Vector:Vector3f=&VectorFromCurve(VectorCurve=DataInterface() {propertyValues {
        XCurve {keys [
          <at 0.25 0.8461 RCIM_Cubic>
          {arriveTangent=-0.4521, interpMode=RCIM_Cubic, leaveTangent=-0.4521, time=0.5, value=0.3291}
          {arriveTangent=-0.6583, interpMode=RCIM_Cubic, leaveTangent=-0.6583, time=0.7, value=0.6427}
          <at 1 0 RCIM_Cubic>
         ]}
        YCurve {keys [
          <at 0.25 0.1038 RCIM_Cubic>
          {arriveTangent=0.1823, interpMode=RCIM_Cubic, leaveTangent=0.1823, time=0.5, value=0.0779}
          {arriveTangent=-0.1559, interpMode=RCIM_Cubic, leaveTangent=-0.1559, time=0.7, value=0.1858}
          <at 1 0 RCIM_Cubic>
         ]}
        ZCurve {keys [
          <at 0.25 1 RCIM_Cubic>
          <at 0.5 0.3906 RCIM_Cubic>
          {arriveTangent=-0.7812, interpMode=RCIM_Cubic, leaveTangent=-0.7812, time=0.7, value=1}
          <at 1 0 RCIM_Cubic>
         ]}
       }})
     Variant 1 Vector:Vector3f=&VectorFromCurve(VectorCurve=DataInterface() {propertyValues {
        XCurve {keys [
          <at 0.25 0.1038 RCIM_Cubic>
          {arriveTangent=0.1819, interpMode=RCIM_Cubic, leaveTangent=0.1819, time=0.5, value=0.0779}
          {arriveTangent=-0.1559, interpMode=RCIM_Cubic, leaveTangent=-0.1559, time=0.701, value=0.1858}
          <at 1 0 RCIM_Cubic>
         ]}
        YCurve {keys [
          <at 0.25 0.7013 RCIM_Cubic>
          {arriveTangent=-0.9984, interpMode=RCIM_Cubic, leaveTangent=-0.9984, time=0.5, value=0.2864}
          {arriveTangent=-0.5728, interpMode=RCIM_Cubic, leaveTangent=-0.5728, time=0.701, value=0.251}
          <at 1 0 RCIM_Cubic>
         ]}
        ZCurve {keys [
          <at 0.25 1 RCIM_Cubic>
          {arriveTangent=-1.5886, interpMode=RCIM_Cubic, leaveTangent=-1.5886, time=0.5, value=0.3906}
          {arriveTangent=-0.7812, interpMode=RCIM_Cubic, leaveTangent=-0.7812, time=0.701, value=0.2836}
          <at 1 0 RCIM_Cubic>
         ]}
       }})
     Variant 2 Vector:Vector3f=?
     Variant 3 Vector:Vector3f=?
     Variant Amount:NiagaraInt32=0
    )
    ScaleA=true
    ScaleRGB=true
   )
   &CameraOffset(Camera Offset Amount=20)
   &DynamicMaterialParameters(
    Index 0 Param 1=&FloatFromCurve(FloatCurve=DataInterface() {propertyValues {Curve {keys [
         <at 0 -0.5 RCIM_Cubic>
         {arriveTangent=18.2487, interpMode=RCIM_Cubic, leaveTangent=18.2487, tangentMode=RCTM_User, time=0.0692, value=0.2738}
         {arriveTangent=4.8641, interpMode=RCIM_Cubic, leaveTangent=4.8641, tangentMode=RCTM_User, time=0.15, value=1}
         {arriveTangent=4.2839, interpMode=RCIM_Cubic, leaveTangent=4.2839, tangentMode=RCTM_User, time=1, value=5}
        ]}}})
    Index 0 Param 3=&FloatFromCurve(FloatCurve=DataInterface() {propertyValues {Curve {keys [
         {arriveTangent=0.1967, interpMode=RCIM_Cubic, leaveTangent=0.1967, tangentMode=RCTM_User, time=0, value=-0.25}
         {arriveTangent=0.1844, interpMode=RCIM_Cubic, leaveTangent=0.1844, tangentMode=RCTM_User, time=1, value=1}
        ]}}})
    Index 0 Param 4=1
    Module.Index0Float0Write=true
    Module.Index0Float1Write=true
    Module.Index0Float2Write=true
    Module.Index0Float3Write=true
   )
  ]
  rendererClasses [&NiagaraSpriteRendererProperties]
  renderers [{rendererClass=&NiagaraSpriteRendererProperties}]
 }
 ElectricArcs {
  emitterUpdateScript [
   &EmitterState(Life Cycle Mode=::NewEnumerator1, Loop Behavior=::NewEnumerator1, Loop Duration=0.2, UseLoopDelay=false)
   &SpawnRate(SpawnRate=50, Use Spawn Probability=false)
   &SpawnBurst_Instantaneous(
    Spawn Count=&RandomRangeInt(Minimum=5, Maximum=6)
    Use Spawn Probability=false
    UseLoopCountLimit=false
   )
   &SpawnRate(
    SpawnGroup=1
    SpawnRate=&FloatFromCurve(FloatCurve=DataInterface() {propertyValues {Curve {keys [<at 0 80 RCIM_Constant>,<at 0.06 0 RCIM_Constant>]}}}, CurveIndex=$Emitter.Age)
    Use Spawn Probability=false
   ) {module=SpawnRate001}
   /SetVariables(Emitter.GlowColor:LinearColor=&VariantColorSelector(Variant Amount=0, Original Color=<rgba 1.378 0.1161 2 0>, Variant 1 Color=<rgba 0 0.6667 2 0>, Variant 2 Color=?)) {isSetParameters=true}
  ]
  particleSpawnScript [
   /SetVariables(Particles.SkeletonSamplingIndex:NiagaraInt32=&SequentialInt(Use Array Data:NiagaraBool=false, First Number:NiagaraInt32=7, Sequence:NiagaraDataInterfaceArrayInt32=?, Last Number:NiagaraInt32=15)) {isSetParameters=true}
   &InitializeParticle(
    Color=&MergeHelper_ColorSwitch(
     0=&VariantColorSelector(Variant Amount=0, Original Color=<rgba 50 0 38.274 0.25>, Variant 1 Color=<rgba 0 12.5 50 0.4>, Variant 2 Color=?)
     1=&VariantColorSelector(
      Variant Amount=0
      Original Color=&RandomRangeLinearColor(Minimum=<rgba 0.6894 1 8 1>, Maximum=<rgba 0 2 4 1>, Color Channel Mode=::NewEnumerator2)
      Variant 1 Color=&RandomRangeLinearColor(Minimum=<rgba 0 2.3529 8 1>, Maximum=<rgba 0 1.1765 4 1>, Color Channel Mode=::NewEnumerator2)
      Variant 2 Color=?
     )
     2=?
     3=?
     Index=$Emitter.SpawnGroup
     Num=::NewEnumerator0
    )
    Lifetime=&MergeHelper_FloatSwitch(
     0=&RandomRangeFloat(Minimum=0.15, Maximum=0.25, Evaluation Type=::NewEnumerator1)
     1=&RandomRangeFloat(Minimum=0.2, Maximum=0.35, Evaluation Type=::NewEnumerator1)
     2=?
     3=?
     Index=$Emitter.SpawnGroup
     Num=::NewEnumerator0
    )
    Sprite Rotation Angle=&MergeHelper_FloatSwitch(
     0=&RandomRangeFloat(Minimum=0.2, Maximum=1)
     1=0
     2=?
     3=?
     Index=$Emitter.SpawnGroup
     Num=::NewEnumerator0
    )
    Sprite Rotation Mode=::NewEnumerator1
    Sprite Size Mode=::NewEnumerator3
    Sprite UV Mode=::NewEnumerator3
    Uniform Sprite Size=&MergeHelper_FloatSwitch(
     0=&RandomRangeFloat(Minimum=25, Maximum=30)
     1=&RandomRangeFloat(Minimum=7, Maximum=10, Evaluation Type=::NewEnumerator1)
     2=?
     3=?
     Index=$Emitter.SpawnGroup
     Num=::NewEnumerator0
    )
    UsePositionOffset=false
    Write Material Random=true
   )
   &SocketLocation(
    Orientation Sampling=::NewEnumerator1
    SamplePosOffset=true
    Sampled Position Offset=&MergeHelper_VectorSwitch(
     0=<0 0 0>
     1=&SelectVectorFromArray(
      Direct Array Index=$Particles.SkeletonSamplingIndex
      Vector Selection Array=DataInterface() {propertyValues {FloatData [<5 0 0>, <5 0 0>, <0 0 0>, <0 0 0>, <-10 0 0>, <10 0 0>, <-10 0 0>, <0 0 0>, <0 0 0>]}}
     )
     2=?
     3=?
     Index=$Emitter.SpawnGroup
     Num=::NewEnumerator0
    )
    Socket Index=&MergeHelper_IntSwitch(
     0=&SequentialInt(Use Array Data:NiagaraBool=false, First Number:NiagaraInt32=0, Sequence:NiagaraDataInterfaceArrayInt32=?, Last Number:NiagaraInt32=6)
     1=$Particles.SkeletonSamplingIndex
     2=?
     3=?
     4=?
     5=?
     Index=$Emitter.SpawnGroup
     Num=::NewEnumerator0
    )
    Socket Reader=$System.Socket Reader
    Socket Sampling Mode=::NewEnumerator3
    VelocitySpeedLimit=false
   )
   &ShapeLocation(
    MaskSpawnGroup=false
    Override Seed=false
    Sphere Radius=&MergeHelper_FloatSwitch(0=10, 1=0, 2=?, 3=?, Index=$Emitter.SpawnGroup, Num=::NewEnumerator0)
    UseOwnerScale=true
   )
   &AddVelocity(
    Module.Use Falloff Distance=false
    Override Seed=false
    Velocity Mode=::NewEnumerator1
    Velocity Origin=$Output.SocketLocation.SampledPosition
    Velocity Speed=&MergeHelper_FloatSwitch(
     0=&RandomRangeFloat(Minimum=1, Maximum=5)
     1=0
     2=?
     3=?
     Index=$Emitter.SpawnGroup
     Num=::NewEnumerator0
    )
   ) {module=AddVelocity001}
   &AddVelocity(Velocity=&MergeHelper_VectorSwitch(
     0=&MakeVector(X=0, Y=0, Z=&RandomRangeFloat(Minimum=25, Maximum=50))
     1=<0 0 0>
     2=?
     3=?
     Index=$Emitter.SpawnGroup
     Num=::NewEnumerator0
    ))
   &CameraOffset(Camera Offset Amount=&MergeHelper_FloatSwitch(
     0=5
     1=&RandomRangeFloat(Minimum=35, Maximum=50, Evaluation Type=::NewEnumerator1)
     2=?
     3=?
     Index=$Emitter.SpawnGroup
     Num=::NewEnumerator0
    ))
   &SubUVAnimation(
    End Frame Range Override=3
    Evaluation Type=::NewEnumerator0
    Output Alternate Renderer Binding=true
    "Sprite Renderer "=DataInterface() {propertyValues {SpriteRenderer=&self:ElectricArcs_0.NiagaraSpriteRendererProperties_12}}
    SubUV Animation Mode:ENiagara_SubUVLookupModeV2=::NewEnumerator2
    UseEndFrame=true
    UseStartFrame=true
   )
   /SetVariables(Particles.VisibilityTag=$Emitter.SpawnGroup) {isSetParameters=true}
  ]
  particleUpdateScript [
   &ParticleState()
   &ScaleSpriteSize(
    Scale Factor=&MergeHelper_Vector2Switch(
     0=&Vector2DFromFloat(Value=&FloatFromCurve(FloatCurve=DataInterface() {propertyValues {Curve {keys [{arriveTangent=2.4335, interpMode=RCIM_Cubic, leaveTangent=2.4335, tangentMode=RCTM_User, time=0, value=0.3}, <at 1 1>]}}}))
     1=&Vector2DFromFloat(Value=&FloatFromCurve(FloatCurve=DataInterface() {propertyValues {Curve {keys [
           <at 0 1 RCIM_Cubic>
           {arriveTangent=1.6706, interpMode=RCIM_Cubic, leaveTangent=1.6706, time=0.043, value=2.109}
           {arriveTangent=-1.1588, interpMode=RCIM_Cubic, leaveTangent=-1.1588, time=0.2029, value=1.339}
           <at 1 1 RCIM_Cubic>
          ]}}}))
     2=?
     3=?
     4=?
     Index=$Particles.VisibilityTag
     Num=::NewEnumerator0
    )
    Scale Sprite Size Mode=::NewEnumerator2
   )
   &ScaleColor(
    Scale Alpha=&MergeHelper_FloatSwitch(
     0=$System.TODScalar
     1=&FloatFromCurve(
      FloatCurve=DataInterface() {propertyValues {Curve {keys [
          {arriveTangent=-0.1447, interpMode=RCIM_Cubic, leaveTangent=-0.1447, tangentMode=RCTM_User, time=0, value=0}
          {arriveTangent=-0.0772, interpMode=RCIM_Cubic, leaveTangent=-0.0772, tangentMode=RCTM_User, time=0.0852, value=1.0091}
          <at 1 0>
         ]}}}
      Scale Curve=$System.TODScalar
     )
     2=?
     3=?
     Index=$Particles.VisibilityTag
     Num=::NewEnumerator0
    )
    Scale RGB=&MergeHelper_VectorSwitch(
     0=<1 1 1>
     1=&VectorFromCurve(VectorCurve=DataInterface() {propertyValues {XCurve {keys [<at 0 0>,<at 1 2>,<at 1 0>]}, YCurve {keys [<at 0 2>,<at 1 0>,<at 1 0>]}, ZCurve {keys [<at 0 0.25>,<at 1 0>,<at 1 0>]}}})
     2=?
     3=?
     Index=$Particles.VisibilityTag
     Num=::NewEnumerator0
    )
    ScaleA=true
    ScaleRGB=true
   )
   &DynamicMaterialParameters(
    Dynamic Parameter Output:ENiagara_AttributeSamplingApplyOutput=::NewEnumerator1
    Index 0 Param 1=&RandomRangeFloat(Minimum=2.5, Maximum=3.5)
    Index 0 Param 2=&FloatFromCurve(FloatCurve=DataInterface() {propertyValues {Curve {keys [<at 0 1>,<at 1 -1>]}}})
    Index 0 Param 3=&RandomRangeFloat(Minimum=0.2, Maximum=0.7)
    Module.Index0Float0Write=true
    Module.Index0Float1Write=true
    Module.Index0Float2Write=true
    Module.Index0Float3Write=true
   )
   &DynamicMaterialParameters(
    Dynamic Parameter Output:ENiagara_AttributeSamplingApplyOutput=::NewEnumerator1
    Index 0 Param 2=&FloatFromCurve(FloatCurve=DataInterface() {propertyValues {Curve {keys [<at 0 1>,<at 1 -1>]}}})
    Module.Index0Float0Write=false
    Module.Index0Float1Write=true
    Module.Index0Float2Write=true
    Module.Index0Float3Write=true
   ) {module=DynamicMaterialParameters001}
   &SolveForcesAndVelocity(Clamp Velocity=false, Limit Acceleration=false)
  ]
  rendererClasses [&NiagaraSpriteRendererProperties]
  renderers [
   {rendererClass=&NiagaraSpriteRendererProperties}
   {rendererClass=&NiagaraSpriteRendererProperties, rendererIndex=1}
  ]
 }
 LightRay_Packed {
  emitterUpdateScript [
   &EmitterState(Life Cycle Mode=::NewEnumerator1, Loop Behavior=::NewEnumerator1, Loop Duration=0.3, UseLoopDelay=false)
   &SpawnRate(SpawnRate=50, Use Spawn Probability=false)
   &SpawnBurst_Instantaneous(Spawn Group=1, Use Spawn Probability=false, UseLoopCountLimit=false)
   &SpawnBurst_Instantaneous(Spawn Group=2, Use Spawn Probability=false, UseLoopCountLimit=false) {module=SpawnBurst_Instantaneous001}
   &SpawnBurst_Instantaneous(Spawn Group=3, Use Spawn Probability=false, UseLoopCountLimit=false) {module=SpawnBurst_Instantaneous002}
   &SpawnBurst_Instantaneous(Spawn Group=4, Use Spawn Probability=false, UseLoopCountLimit=false) {module=SpawnBurst_Instantaneous003}
  ]
  particleSpawnScript [
   &InitializeParticle(
    Color=&VariantColorSelector(
     Variant Amount=0
     Original Color=&MergeHelper_ColorSwitch(
      0=<rgba 2 2 2 0.5>
      1=<rgba 2 2 2 0.25>
      2=<rgba 2 2 2 0.25>
      3=<rgba 2 2 2 0.25>
      4=<rgba 15 15 15 0.125>
      Index=$Emitter.SpawnGroup
      Num=::NewEnumerator3
     )
     Variant 1 Color=&MergeHelper_ColorSwitch(
      0=<rgba 2 2 2 0.75>
      1=<rgba 2 2 2 0.25>
      2=<rgba 2 2 2 0.5>
      3=<rgba 2 2 2 0.25>
      4=<rgba 15 15 15 0.2>
      Index=$Emitter.SpawnGroup
      Num=::NewEnumerator3
     )
     Variant 2 Color=?
    )
    Lifetime=&MergeHelper_FloatSwitch(
     0=&RandomRangeFloat(Minimum=0.25, Maximum=0.3, Evaluation Type=::NewEnumerator1)
     1=0.25
     2=0.3
     3=0.35
     4=&RandomRangeFloat(Minimum=0.35, Maximum=0.45, Evaluation Type=::NewEnumerator1)
     Index=$Emitter.SpawnGroup
     Num=::NewEnumerator3
    )
    Sprite Rotation Angle=&MergeHelper_FloatSwitch(
     0=&RandomRangeFloat(Minimum=0, Maximum=1)
     1=0
     2=0
     3=0
     4=0
     Index=$Emitter.SpawnGroup
     Num=::NewEnumerator3
    )
    Sprite Rotation Mode=::NewEnumerator1
    Sprite Size=&MergeHelper_Vector2Switch(
     0=&RandomRangeVector2D(Minimum=<50 100>, Maximum=<100 250>, Evaluation Type=::NewEnumerator1)
     1=<0 0>
     2=&Vector2DFromFloat(Value=&RandomRangeFloat(Minimum=300, Maximum=400, Evaluation Type=::NewEnumerator1))
     3=<0 0>
     4=&RandomRangeVector2D(Minimum=<105 200>, Maximum=<140 250>, Evaluation Type=::NewEnumerator1)
     Index=$Emitter.SpawnGroup
     Num=::NewEnumerator3
    )
    Sprite Size Mode=::NewEnumerator4
    Sprite UV Mode=::NewEnumerator1
    UsePositionOffset=false
    Write Material Random=true
   )
   &SocketLocation(
    SamplePosOffset=true
    Sampled Position Offset=&MergeHelper_VectorSwitch(0=<0 0 0>, 1=<0 0 10>, 2=<0 0 0>, 3=<0 0 0>, 4=<0 0 -35>, Index=$Emitter.SpawnGroup, Num=::NewEnumerator3)
    Sampled Position Offset Coordinate Space:ENiagaraSkelMeshTransforms=::NewEnumerator2
    Socket Index=&MergeHelper_IntSwitch(0=6, 1=16, 2=6, 3=6, 4=6, 5=?, Index=$Emitter.SpawnGroup, Num=::NewEnumerator3)
    Socket Reader=$System.Socket Reader
    Socket Sampling Mode=::NewEnumerator3
    VelocitySpeedLimit=false
   )
   &ShapeLocation(
    MaskSpawnGroup=false
    Override Seed=false
    Sphere Radius=&MergeHelper_FloatSwitch(0=2, 1=0, 2=2, 3=0, 4=0, Index=$Emitter.SpawnGroup, Num=::NewEnumerator3)
    UseOwnerScale=true
   )
   &SubUVAnimation(
    End Frame Range Override=3
    Evaluation Type=::NewEnumerator0
    Output Alternate Renderer Binding=true
    "Sprite Renderer "=DataInterface() {propertyValues {SpriteRenderer=&self:LightRay_Packed_0.NiagaraSpriteRendererProperties_1}}
    SubUV Animation Mode:ENiagara_SubUVLookupModeV2=::NewEnumerator2
    UseEndFrame=true
    UseStartFrame=true
   )
   &AddVelocity(Velocity=&MergeHelper_VectorSwitch(0=<0 0 0>, 1=<0 0 220>, 2=<0 0 0>, 3=<0 0 0>, 4=<0 0 0>, Index=$Emitter.SpawnGroup, Num=::NewEnumerator3))
   /SetVariables(Particles.VisibilityTag=$Emitter.SpawnGroup) {isSetParameters=true}
  ]
  particleUpdateScript [
   &ParticleState()
   &ScaleSpriteSize(
    Scale Factor=&MergeHelper_Vector2Switch(
     0=&Vector2DFromCurve(
      Vector2Curve=DataInterface() {propertyValues {
        XCurve {keys [
          {arriveTangent=4.9957, interpMode=RCIM_Cubic, leaveTangent=4.9957, tangentMode=RCTM_User, time=-0.0026, value=1}
          {arriveTangent=0.9974, interpMode=RCIM_Cubic, leaveTangent=0.9974, time=0.3904, value=1.8714}
          <at 1 2 RCIM_Cubic>
         ]}
        YCurve {keys [
          {arriveTangent=0.3115, interpMode=RCIM_Cubic, leaveTangent=0.3115, tangentMode=RCTM_User, time=-0.0026, value=0.8481}
          {arriveTangent=0.1515, interpMode=RCIM_Cubic, leaveTangent=0.1515, time=0.3904, value=0.9697}
          <at 1 1 RCIM_Cubic>
         ]}
       }}
      ScaleCurve=<1 1>
     )
     1=<0 0>
     2=&Vector2DFromFloat(Value=&FloatFromCurve(FloatCurve=DataInterface() {propertyValues {Curve {keys [
           {arriveTangent=-0.1064, interpMode=RCIM_Cubic, leaveTangent=-0.1064, tangentMode=RCTM_User, time=0.0019, value=0.5035}
           {arriveTangent=-0.007, interpMode=RCIM_Cubic, leaveTangent=-0.007, time=0.5, value=1}
           <at 1.0038 0.4965 RCIM_Cubic>
          ]}}}))
     3=<0 0>
     4=&Vector2DFromCurve(
      Vector2Curve=DataInterface() {propertyValues {
        XCurve {keys [
          {arriveTangent=2.4707, interpMode=RCIM_Cubic, leaveTangent=2.4707, tangentMode=RCTM_User, time=-0.0026, value=1}
          <at 0.1418 1 RCIM_Cubic>
          <at 1 1 RCIM_Cubic>
         ]}
        YCurve {keys [
          {interpMode=RCIM_Cubic, tangentMode=RCTM_User, time=-0.0026, value=0.8481}
          {arriveTangent=0.1515, interpMode=RCIM_Cubic, leaveTangent=0.1515, time=0.1418, value=0.9674}
          <at 1 1 RCIM_Cubic>
         ]}
       }}
      ScaleCurve=<1 1>
     )
     Index=$Particles.VisibilityTag
     Num=::NewEnumerator3
    )
    Scale Sprite Size Mode=::NewEnumerator2
   )
   &ScaleColor(
    Scale Alpha=&FloatFromCurve(
     FloatCurve=DataInterface() {propertyValues {Curve {keys [
         {arriveTangent=-0.1447, interpMode=RCIM_Cubic, leaveTangent=-0.1447, tangentMode=RCTM_User, time=0, value=0}
         {arriveTangent=-0.0772, interpMode=RCIM_Cubic, leaveTangent=-0.0772, tangentMode=RCTM_User, time=0.0186, value=1.0006}
         <at 1 0>
        ]}}}
     Scale Curve=$System.TODScalar
    )
    Scale RGB=&VariantVectorSelector(
     Original Vector:Vector3f=&VectorFromCurve(VectorCurve=DataInterface() {propertyValues {
        XCurve {keys [
          {arriveTangent=-0.139, time=0.25, value=0.3097}
          {arriveTangent=-0.1114, interpMode=RCIM_Cubic, leaveTangent=-0.1114, time=0.6, value=0.2777}
          {arriveTangent=-1.1392, leaveTangent=-1.1392, time=0.6738, value=0.2624}
          <at 1 0>
         ]}
        YCurve {keys [
          {arriveTangent=0.0234, time=0.25, value=0.0523}
          {arriveTangent=0.1447, interpMode=RCIM_Cubic, leaveTangent=0.1447, time=0.6, value=0.0868}
          {arriveTangent=-0.4345, leaveTangent=-0.4345, time=0.6738, value=0.1136}
          <at 1 0>
         ]}
        ZCurve {keys [
          {arriveTangent=0.0476, time=0.25, value=0.4167}
          {arriveTangent=1.3765, interpMode=RCIM_Cubic, leaveTangent=1.3765, time=0.6, value=0.3906}
          {arriveTangent=-2.0374, leaveTangent=-2.0374, time=0.6738, value=1}
          <at 1 0>
         ]}
       }})
     Variant 1 Vector:Vector3f=&VectorFromCurve(VectorCurve=DataInterface() {propertyValues {
        XCurve {keys [{arriveTangent=-1.1392, time=0.6738, value=0.1136}, <at 1 0>]}
        YCurve {keys [{arriveTangent=-0.4345, time=0.6738, value=0.3341}, <at 1 0>]}
        ZCurve {keys [{arriveTangent=-2.0374, time=0.6738, value=0.6}, <at 1 0>]}
       }})
     Variant 2 Vector:Vector3f=?
     Variant 3 Vector:Vector3f=?
     Variant Amount:NiagaraInt32=0
    )
    ScaleA=true
    ScaleRGB=true
   )
   &CameraOffset(Camera Offset Amount=&MergeHelper_FloatSwitch(0=10, 1=50, 2=50, 3=25, 4=50, Index=$Particles.VisibilityTag, Num=::NewEnumerator3))
   &FN_Light_Attributes(
    Color=&MakeLinearColorFromVectorAndFloat(
     "Vector (RGB)"=&VariantVectorSelector(
      Original Vector:Vector3f=&VectorFromCurve(
       VectorCurve=DataInterface() {propertyValues {
         XCurve {keys [
           {arriveTangent=-0.139, time=0.25, value=0.3097}
           {arriveTangent=-0.1114, interpMode=RCIM_Cubic, leaveTangent=-0.1114, time=0.6, value=0.2777}
           {arriveTangent=-1.1392, leaveTangent=-1.1392, time=0.6738, value=0.2624}
           <at 1 0>
          ]}
         YCurve {keys [
           {arriveTangent=0.0234, time=0.25, value=0.0523}
           {arriveTangent=0.1447, interpMode=RCIM_Cubic, leaveTangent=0.1447, time=0.6, value=0.0868}
           {arriveTangent=-0.4345, leaveTangent=-0.4345, time=0.6738, value=0.1136}
           <at 1 0>
          ]}
         ZCurve {keys [
           {arriveTangent=0.0476, time=0.25, value=0.4167}
           {arriveTangent=1.3765, interpMode=RCIM_Cubic, leaveTangent=1.3765, time=0.6, value=0.3906}
           {arriveTangent=-2.0374, leaveTangent=-2.0374, time=0.6738, value=1}
           <at 1 0>
          ]}
        }}
       Scale Curve=<15 15 15>
      )
      Variant 1 Vector:Vector3f=<0 7.5 15>
      Variant 2 Vector:Vector3f=?
      Variant 3 Vector:Vector3f=?
      Variant Amount:NiagaraInt32=0
     )
     "Float (Alpha)"=&FloatFromCurve(
      FloatCurve=DataInterface() {propertyValues {Curve {keys [{arriveTangent=-2.1518, interpMode=RCIM_Cubic, leaveTangent=-2.1518, tangentMode=RCTM_User, time=0, value=1.0006}, <at 1 0>]}}}
      Scale Curve=8
     )
    )
    Create Color Attribute=true
    Create Exponent Attribute=true
    Create Radius Attribute=true
    Exponent=5
    Radius=100
   )
   &DynamicMaterialParameters(
    Dynamic Parameter Output:ENiagara_AttributeSamplingApplyOutput=::NewEnumerator1
    Index 0 Param 1=1
    Index 0 Param 2=1
    Index 0 Param 3=1
    Index 0 Param 4=1
    Module.Index0Float0Write=true
    Module.Index0Float1Write=true
    Module.Index0Float2Write=true
    Module.Index0Float3Write=true
   )
   &DynamicMaterialParameters(
    Dynamic Parameter Output:ENiagara_AttributeSamplingApplyOutput=::NewEnumerator1
    Index 0 Param 1=0.85
    Index 0 Param 2=1
    Index 0 Param 3=15
    Index 0 Param 4=1
    Module.Index0Float0Write=true
    Module.Index0Float1Write=true
    Module.Index0Float2Write=true
    Module.Index0Float3Write=true
   ) {module=DynamicMaterialParameters001}
   &FN_Mesh_Attributes(
    Color=&MakeLinearColorFromVectorAndFloat(
     "Vector (RGB)"=&VariantVectorSelector(
      Original Vector:Vector3f=&VectorFromCurve(VectorCurve=DataInterface() {propertyValues {
         XCurve {keys [
           <at 0.25 0.8461 RCIM_Cubic>
           {arriveTangent=-0.4521, interpMode=RCIM_Cubic, leaveTangent=-0.4521, time=0.5, value=0.3291}
           {arriveTangent=-0.6583, interpMode=RCIM_Cubic, leaveTangent=-0.6583, time=0.7, value=0.6427}
           <at 1 0 RCIM_Cubic>
          ]}
         YCurve {keys [
           <at 0.25 0.1038 RCIM_Cubic>
           {arriveTangent=0.1823, interpMode=RCIM_Cubic, leaveTangent=0.1823, time=0.5, value=0.0779}
           {arriveTangent=-0.1559, interpMode=RCIM_Cubic, leaveTangent=-0.1559, time=0.7, value=0.1858}
           <at 1 0 RCIM_Cubic>
          ]}
         ZCurve {keys [
           <at 0.25 1 RCIM_Cubic>
           <at 0.5 0.3906 RCIM_Cubic>
           {arriveTangent=-0.7812, interpMode=RCIM_Cubic, leaveTangent=-0.7812, time=0.7, value=1}
           <at 1 0 RCIM_Cubic>
          ]}
        }})
      Variant 1 Vector:Vector3f=&VectorFromCurve(VectorCurve=DataInterface() {propertyValues {
         XCurve {keys [
           <at 0.25 0 RCIM_Cubic>
           <at 0.5 0.2917 RCIM_Cubic>
           {arriveTangent=-0.5833, interpMode=RCIM_Cubic, leaveTangent=-0.5833, time=0.7, value=0}
           <at 1 0 RCIM_Cubic>
          ]}
         YCurve {keys [
           <at 0.25 0.3401 RCIM_Cubic>
           {arriveTangent=-0.248, interpMode=RCIM_Cubic, leaveTangent=-0.248, time=0.5, value=0.5325}
           {arriveTangent=-1.0651, interpMode=RCIM_Cubic, leaveTangent=-1.0651, time=0.7, value=0.2285}
           <at 1 0 RCIM_Cubic>
          ]}
         ZCurve {keys [
           <at 0.25 1 RCIM_Cubic>
           {arriveTangent=-0.787, interpMode=RCIM_Cubic, leaveTangent=-0.787, time=0.5, value=1}
           {arriveTangent=-2, interpMode=RCIM_Cubic, leaveTangent=-2, time=0.7, value=0.6458}
           <at 1 0 RCIM_Cubic>
          ]}
        }})
      Variant 2 Vector:Vector3f=?
      Variant 3 Vector:Vector3f=?
      Variant Amount:NiagaraInt32=0
     )
     "Float (Alpha)"=&FloatFromCurve(
      FloatCurve=DataInterface() {propertyValues {Curve {keys [
          {arriveTangent=-0.1447, interpMode=RCIM_Cubic, leaveTangent=-0.1447, tangentMode=RCTM_User, time=0, value=0}
          {arriveTangent=-0.0772, interpMode=RCIM_Cubic, leaveTangent=-0.0772, tangentMode=RCTM_User, time=0.0919, value=0.2487}
          <at 1 0>
         ]}}}
      Scale Curve=&Multiply_Float(
       A=/BRCosmetics/Weapons/FORT_Melee/Pickaxe_GateHound/FX/VariantManager/VariantFloatSelector(
        Original Float:NiagaraFloat=0.085
        Variant 1 Float:NiagaraFloat=0.2
        Variant 2 Float:NiagaraFloat=?
        Variant 3 Float:NiagaraFloat=?
        Variant Amount:NiagaraInt32=0
       )
       B=$System.TODScalar
      )
     )
    )
    Create Color Attribute=true
    Create Scale Attribute=true
    Module Output=::NewEnumerator1
    Scale=<1.25 1.25 1.25>
   )
   &SolveForcesAndVelocity(Clamp Velocity=false, Limit Acceleration=false)
  ]
  rendererClasses [
   &NiagaraSpriteRendererProperties
   &NiagaraLightRendererProperties
   &NiagaraMeshRendererProperties
  ]
  renderers [
   {rendererClass=&NiagaraSpriteRendererProperties}
   {rendererClass=&NiagaraLightRendererProperties, rendererIndex=1}
   {rendererClass=&NiagaraSpriteRendererProperties, rendererIndex=2}
   {rendererClass=&NiagaraMeshRendererProperties, rendererIndex=3}
   {rendererClass=&NiagaraSpriteRendererProperties, rendererIndex=4}
  ]
 }
 SpiralSmoke {
  emitterUpdateScript [
   &EmitterState()
   &SpawnRate(SpawnRate=125, Use Spawn Probability=false)
   &SpawnRate(
    Spawn Probability=&VariantFloatSelector(
     Original Float:NiagaraFloat=0
     Variant 1 Float:NiagaraFloat=1
     Variant 2 Float:NiagaraFloat=?
     Variant 3 Float:NiagaraFloat=?
     Variant Amount:NiagaraInt32=0
    )
    SpawnGroup=1
    SpawnRate=80
    Use Spawn Probability=true
   ) {module=SpawnRate001}
  ]
  particleSpawnScript [
   &PartitionParticles()
   &InitializeParticle(
    Color=&VariantColorSelector(
     Variant Amount=0
     Original Color=&RandomRangeLinearColor(Minimum=<rgba 2 2 2 0.8>, Maximum=<rgba 15 15 15 1>)
     Variant 1 Color=&RandomRangeLinearColor(Minimum=<rgba 2 2 2 1.5>, Maximum=<rgba 15 15 15 1>)
     Variant 2 Color=?
    )
    Lifetime=&MergeHelper_FloatSwitch(
     0=&RandomRangeFloat(Minimum=0.5, Maximum=0.9)
     1=&RandomRangeFloat(Minimum=0.18, Maximum=0.3)
     2=?
     3=?
     Index=$Emitter.SpawnGroup
     Num=::NewEnumerator0
    )
    Sprite Rotation Mode=::NewEnumerator2
    Sprite Size Mode=::NewEnumerator3
    Sprite UV Mode=::NewEnumerator1
    Uniform Sprite Size=&RandomRangeFloat(Minimum=8.4, Maximum=10.5, Evaluation Type=::NewEnumerator1)
    UsePositionOffset=false
    Write Material Random=true
   )
   &SocketLocation(
    SamplePosOffset=true
    Sampled Position Offset=&RandomRangeVector(Minimum=<0 0 0>, Maximum=<0 0 25>)
    Sampled Position Offset Coordinate Space:ENiagaraSkelMeshTransforms=::NewEnumerator2
    Socket Index=&MergeHelper_IntSwitch(
     0=&SequentialInt(Use Array Data:NiagaraBool=false, First Number:NiagaraInt32=0, Sequence:NiagaraDataInterfaceArrayInt32=?, Last Number:NiagaraInt32=6)
     1=&SequentialInt(Use Array Data:NiagaraBool=false, First Number:NiagaraInt32=17, Sequence:NiagaraDataInterfaceArrayInt32=?, Last Number:NiagaraInt32=21)
     2=?
     3=?
     4=?
     5=?
     Index=$Emitter.SpawnGroup
     Num=::NewEnumerator0
    )
    Socket Reader=$System.Socket Reader
    Socket Sampling Mode=::NewEnumerator3
    VelocitySpeedLimit=false
   )
   &ShapeLocation(MaskSpawnGroup=false, Override Seed=false, Sphere Radius=2, UseOwnerScale=true)
   &AddVelocity(
    Module.Use Falloff Distance=false
    Override Seed=false
    Velocity Mode=::NewEnumerator1
    Velocity Origin=$Output.SocketLocation.SampledPosition
    Velocity Speed=2
   ) {module=AddVelocity001}
   &AddVelocity(Velocity=&MergeHelper_VectorSwitch(
     0=&MakeVector(X=0, Y=0, Z=&RandomRangeFloat(Minimum=25, Maximum=80))
     1=&MakeVector(X=0, Y=0, Z=&RandomRangeFloat(Minimum=25, Maximum=100))
     2=?
     3=?
     Index=$Emitter.SpawnGroup
     Num=::NewEnumerator0
    ))
   &DynamicMaterialParameters(
    Dynamic Parameter Output:ENiagara_AttributeSamplingApplyOutput=::NewEnumerator1
    Index 0 Param 1=&RandomRangeFloat(Minimum=0, Maximum=4, Evaluation Type=::NewEnumerator1)
    Index 0 Param 2=&RandomRangeFloat(Minimum=1, Maximum=4, Evaluation Type=::NewEnumerator1)
    Index 0 Param 3=120
    Index 0 Param 4=0.5
    Module.Index0Float0Write=true
    Module.Index0Float1Write=true
    Module.Index0Float2Write=true
    Module.Index0Float3Write=true
   )
   &SubUVAnimation(
    End Frame Range Override=3
    Evaluation Type=::NewEnumerator0
    Output Alternate Renderer Binding=true
    "Sprite Renderer "=DataInterface() {propertyValues {SpriteRenderer=&self:SpiralSmoke_0.NiagaraSpriteRendererProperties_5}}
    SubUV Animation Mode:ENiagara_SubUVLookupModeV2=::NewEnumerator2
    UseEndFrame=true
    UseStartFrame=true
   ) {module=SubUVAnimation001}
   /SetVariables(Particles.VisibilityTag=$Emitter.SpawnGroup) {isSetParameters=true}
  ]
  particleUpdateScript [
   &ParticleState()
   &Drag(Drag=2, Use Linear Drag=true, Use Rotational Drag=true)
   &ScaleColor(
    Scale Alpha=&FloatFromCurve(
     FloatCurve=DataInterface() {propertyValues {Curve {keys [
         {arriveTangent=-0.1447, interpMode=RCIM_Cubic, leaveTangent=-0.1447, tangentMode=RCTM_User, time=0, value=0}
         {arriveTangent=-0.0772, interpMode=RCIM_Cubic, leaveTangent=-0.0772, tangentMode=RCTM_User, time=0.0186, value=1.0006}
         <at 1 0>
        ]}}}
     Scale Curve=$System.TODScalar
    )
    Scale RGB=&VariantVectorSelector(
     Original Vector:Vector3f=&VectorFromCurve(VectorCurve=DataInterface() {propertyValues {
        XCurve {keys [
          {arriveTangent=-0.139, time=0.25, value=0.3097}
          {arriveTangent=-0.1114, interpMode=RCIM_Cubic, leaveTangent=-0.1114, time=0.6, value=0.2777}
          {arriveTangent=-1.1392, leaveTangent=-1.1392, time=0.6738, value=0.2624}
          <at 1 0>
         ]}
        YCurve {keys [
          {arriveTangent=0.0234, time=0.25, value=0.0523}
          {arriveTangent=0.1447, interpMode=RCIM_Cubic, leaveTangent=0.1447, time=0.6, value=0.0868}
          {arriveTangent=-0.4345, leaveTangent=-0.4345, time=0.6738, value=0.1136}
          <at 1 0>
         ]}
        ZCurve {keys [
          {arriveTangent=0.0476, time=0.25, value=0.4167}
          {arriveTangent=1.3765, interpMode=RCIM_Cubic, leaveTangent=1.3765, time=0.6, value=0.3906}
          {arriveTangent=-2.0374, leaveTangent=-2.0374, time=0.6738, value=1}
          <at 1 0>
         ]}
       }})
     Variant 1 Vector:Vector3f=&VectorFromCurve(VectorCurve=DataInterface() {propertyValues {
        XCurve {keys [{arriveTangent=-0.139, time=0.25, value=0}, <at 0.6 0 RCIM_Cubic>, <at 1 0>]}
        YCurve {keys [
          {arriveTangent=0.0234, time=0.25, value=0.217}
          {arriveTangent=-0.2894, interpMode=RCIM_Cubic, leaveTangent=-0.2894, time=0.6, value=0.1302}
          <at 1 0>
         ]}
        ZCurve {keys [
          {arriveTangent=0.0476, time=0.25, value=0.5208}
          {arriveTangent=-0.6944, interpMode=RCIM_Cubic, leaveTangent=-0.6944, time=0.6, value=0.3906}
          <at 1 0>
         ]}
       }})
     Variant 2 Vector:Vector3f=?
     Variant 3 Vector:Vector3f=?
     Variant Amount:NiagaraInt32=0
    )
    ScaleA=true
    ScaleRGB=true
   )
   &ScaleSpriteSize(
    Scale Sprite Size Mode=::NewEnumerator0
    Uniform Scale Factor=&FloatFromCurve(FloatCurve=DataInterface() {propertyValues {Curve {keys [
         {arriveTangent=16.9657, interpMode=RCIM_Cubic, leaveTangent=16.9657, tangentMode=RCTM_User, time=0, value=1}
         {arriveTangent=4.2114, interpMode=RCIM_Cubic, leaveTangent=4.2114, time=0.2637, value=3.9734}
         <at 1.0038 5.2275 RCIM_Cubic>
        ]}}})
   )
   &AccelerationForce(Acceleration=&MakeVector(X=0, Y=&RandomRangeFloat(Minimum=-25, Maximum=-35), Z=2))
   &SolveForcesAndVelocity(Clamp Velocity=false, Limit Acceleration=false)
   &SubUVAnimation(
    End Frame Range Override=62
    "Sprite Renderer "=DataInterface() {propertyValues {SpriteRenderer=&self:SpiralSmoke_0.NiagaraSpriteRendererProperties_1}}
    Start Frame Range Override=30
    UseEndFrame=true
    UseStartFrame=true
   )
   &CameraOffset(Camera Offset Amount=&FloatFromCurve(FloatCurve=DataInterface() {propertyValues {Curve {keys [<at 0 15>,<at 0.5 -5>]}}}))
   &FN_Sprite_Attributes(
    Camera Offset:NiagaraFloat=&FloatFromCurve(FloatCurve=DataInterface() {propertyValues {Curve {keys [<at 0 -5>,<at 1 17>]}}})
    Color=&MakeLinearColorFromVectorAndFloat(
     "Vector (RGB)"=&Multiply_VectorByFloat(
      Vector=&MakeVectorFromLinearColorRGB(Color=&MergeHelper_ColorSwitch(
        0=&RandomRangeLinearColor(Minimum=<rgba 0 0.3333 1 0.1>, Maximum=<rgba 0 0.3333 1 0.15>)
        1=&RandomRangeLinearColor(Minimum=<rgba 0 0.3333 1 0.005>, Maximum=<rgba 0 0.3333 1 0.01>)
        2=?
        3=?
        Index=$Particles.PartitionParticles.Partition
        Num=::NewEnumerator0
       ))
      Float=&FloatFromCurve(
       FloatCurve=DataInterface() {propertyValues {Curve {keys [<at 0 10 RCIM_Cubic>,<at 1 2>]}}}
       Scale Curve=&MergeHelper_FloatSwitch(0=1, 1=0.5, 2=?, 3=?, Index=$Particles.PartitionParticles.Partition, Num=::NewEnumerator0)
      )
     )
     "Float (Alpha)"=&FloatFromCurve(FloatCurve=DataInterface() {propertyValues {Curve {keys [<at 0 1 RCIM_Cubic>,<at 1 0>]}}}, Scale Curve=$System.TODScalar)
    )
    Create Camera Offset Attribute=true
    Create Color Attribute=true
    Create SpriteSize Attribute=true
    Module Output:ENiagara_AttributeSamplingApplyOutput=::NewEnumerator1
    SpriteSize=&Multiply_Vector2D(
     A=&MergeHelper_Vector2Switch(
      0=&RandomRangeVector2D(Minimum=<9 12>, Maximum=<12 18>)
      1=&RandomRangeVector2D(Minimum=<12 20>, Maximum=<15 28>)
      2=?
      3=?
      4=?
      Index=$Particles.PartitionParticles.Partition
      Num=::NewEnumerator0
     )
     B=&MergeHelper_Vector2Switch(
      0=&Vector2DFromCurve(
       Vector2Curve=DataInterface() {propertyValues {
         XCurve {keys [{arriveTangent=0.5317, interpMode=RCIM_Cubic, leaveTangent=0.5317, tangentMode=RCTM_User, time=-0.0045, value=1}, <at 1 0.4>]}
         YCurve {keys [{arriveTangent=3.7386, interpMode=RCIM_Cubic, leaveTangent=3.7386, tangentMode=RCTM_User, time=-0.0045, value=0.05}, <at 1 2>]}
        }}
       ScaleCurve=<1 1>
      )
      1=&Vector2DFromCurve(
       Vector2Curve=DataInterface() {propertyValues {
         XCurve {keys [{arriveTangent=0.5317, interpMode=RCIM_Cubic, leaveTangent=0.5317, tangentMode=RCTM_User, time=-0.0045, value=1}, <at 1 0.4>]}
         YCurve {keys [{arriveTangent=3.7386, interpMode=RCIM_Cubic, leaveTangent=3.7386, tangentMode=RCTM_User, time=-0.0045, value=0.05}, <at 1 3>]}
        }}
       ScaleCurve=<1 1>
      )
      2=?
      3=?
      4=?
      Index=$Particles.PartitionParticles.Partition
      Num=::NewEnumerator0
     )
    )
   )
  ]
  rendererClasses [&NiagaraSpriteRendererProperties]
  renderers [
   {rendererClass=&NiagaraSpriteRendererProperties}
   {rendererClass=&NiagaraSpriteRendererProperties, rendererIndex=1}
  ]
 }
 SpiralSmoke_Packed {
  emitterUpdateScript [
   &EmitterState(Life Cycle Mode=::NewEnumerator1, Loop Behavior=::NewEnumerator1, Loop Duration=1, UseLoopDelay=false)
   &SpawnBurst_Instantaneous(Spawn Count=4, Use Spawn Probability=false, UseLoopCountLimit=false)
   &SpawnBurst_Instantaneous(Spawn Group=1, Use Spawn Probability=false, UseLoopCountLimit=false) {module=SpawnBurst_Instantaneous001}
   &SpawnBurst_Instantaneous(Spawn Group=2, Use Spawn Probability=false, UseLoopCountLimit=false) {module=SpawnBurst_Instantaneous002}
   /SetVariables(
    Emitter.ColorHigh:LinearColor=&VariantColorSelector(Variant Amount=0, Original Color=<rgba 1.1722 0.3529 10 1>, Variant 1 Color=<rgba 0 0.3333 1 1>, Variant 2 Color=?)
    Emitter.ColorLow:LinearColor=&VariantColorSelector(Variant Amount=0, Original Color=<rgba 30 2.6192 26.3278 1>, Variant 1 Color=<rgba 0.5781 1.4349 3 1>, Variant 2 Color=?)
   ) {isSetParameters=true}
   /SetVariables(Emitter.GlowColor:LinearColor=&VariantColorSelector(Variant Amount=0, Original Color=<rgba 0.8963 0.011 1 1>, Variant 1 Color=<rgba 0.5549 1.4199 3 1>, Variant 2 Color=?)) {isSetParameters=true}
  ]
  particleSpawnScript [
   &InitializeParticle(
    Color=&MergeHelper_ColorSwitch(
     0=&VariantColorSelector(Variant Amount=0, Original Color=<rgba 15 15 15 0.005>, Variant 1 Color=<rgba 2 2 2 1>, Variant 2 Color=?)
     1=&VariantColorSelector(Variant Amount=0, Original Color=<rgba 5 5 5 0.25>, Variant 1 Color=<rgba 5 5 5 0.7>, Variant 2 Color=?)
     2=&VariantColorSelector(Variant Amount=0, Original Color=<rgba 1 1 1 0.35>, Variant 1 Color=<rgba 1 1 1 1>, Variant 2 Color=?)
     3=?
     Index=$Emitter.SpawnGroup
     Num=::NewEnumerator1
    )
    Lifetime=&MergeHelper_FloatSwitch(
     0=&RandomRangeFloat(Minimum=0.5, Maximum=1.2)
     1=0.25
     2=0.75
     3=?
     Index=$Emitter.SpawnGroup
     Num=::NewEnumerator1
    )
    Mesh Scale=&MergeHelper_VectorSwitch(
     0=&RandomRangeVector(Minimum=<0.245 0.245 0.35>, Maximum=<0.28 0.28 1.4>)
     1=<0.7 0.7 0.7>
     2=<1 1 1>
     3=?
     Index=$Emitter.SpawnGroup
     Num=::NewEnumerator1
    )
    Mesh Scale Mode=::NewEnumerator4
    UsePositionOffset=false
    Write Material Random=true
   )
   &SocketLocation(
    Orientation Sampling=::NewEnumerator1
    SamplePosOffset=true
    Sampled Position Offset=&MergeHelper_VectorSwitch(0=<0 0 0>, 1=<0 0 0>, 2=<-10 0 0>, 3=?, Index=$Emitter.SpawnGroup, Num=::NewEnumerator1)
    Socket Index=&MergeHelper_IntSwitch(
     0=&SequentialInt(Use Array Data:NiagaraBool=false, First Number:NiagaraInt32=0, Sequence:NiagaraDataInterfaceArrayInt32=?, Last Number:NiagaraInt32=6)
     1=6
     2=6
     3=?
     4=?
     5=?
     Index=$Emitter.SpawnGroup
     Num=::NewEnumerator1
    )
    Socket Reader=$System.Socket Reader
    Socket Sampling Mode=::NewEnumerator3
    VelocitySpeedLimit=false
   )
   &InitialMeshOrientation(
    Mesh Orientation Mode=::NewEnumerator7
    Rotation=&MergeHelper_VectorSwitch(
     0=&MakeVector(X=0, Y=0, Z=&RandomRangeFloat(Minimum=-1, Maximum=1))
     1=<0 0.25 0>
     2=<0 0 0>
     3=?
     Index=$Emitter.SpawnGroup
     Num=::NewEnumerator1
    )
    Use Rotation Vector=true
   )
   &AddVelocity(Velocity=&MergeHelper_VectorSwitch(0=<0 0 0>, 1=<0 0 0>, 2=<0 0 50>, 3=?, Index=$Emitter.SpawnGroup, Num=::NewEnumerator1))
   /SetVariables(Particles.VisibilityTag=$Emitter.SpawnGroup) {isSetParameters=true}
  ]
  particleUpdateScript [
   &ParticleState()
   &ScaleMeshSize(Scale Factor=&MergeHelper_VectorSwitch(
     0=&VectorFromCurve(VectorCurve=DataInterface() {propertyValues {XCurve {keys [<at 0 1 RCIM_Cubic>,<at 1 1.1>]}, YCurve {keys [<at 0 1 RCIM_Cubic>,<at 1 1.1>]}, ZCurve {keys [<at 0 1 RCIM_Cubic>,<at 1 1>]}}})
     1=&VectorFromCurve(VectorCurve=DataInterface() {propertyValues {XCurve {keys [<at 0 0.2>,<at 1 1>]}, YCurve {keys [<at 0 0.2>,<at 1 1>]}, ZCurve {keys [<at 0 0.2>,<at 1 1>]}}})
     2=&VectorFromCurve(VectorCurve=DataInterface() {propertyValues {XCurve {keys [<at 0 1>,<at 1 1>]}, YCurve {keys [<at 0 1>,<at 1 3.01>]}, ZCurve {keys [<at 0 1>,<at 1 2>]}}})
     3=?
     Index=$Particles.VisibilityTag
     Num=::NewEnumerator1
    ))
   &DynamicMaterialParameters(
    Dynamic Parameter Output:ENiagara_AttributeSamplingApplyOutput=::NewEnumerator1
    Index 0 Param 1=&RandomRangeFloat(Minimum=0, Maximum=1)
    Index 0 Param 2=&FloatFromCurve(FloatCurve=DataInterface() {propertyValues {Curve {keys [{arriveTangent=0.7124, time=-0.0026, value=0.2876}, <at 1 1 RCIM_Cubic>]}}})
    Index 0 Param 3=1
    Index 0 Param 4=1
    Module.Index0Float0Write=true
    Module.Index0Float1Write=true
    Module.Index0Float2Write=true
    Module.Index0Float3Write=true
   )
   &DynamicMaterialParameters(
    Dynamic Parameter Output:ENiagara_AttributeSamplingApplyOutput=::NewEnumerator1
    Index 0 Param 1=&FloatFromCurve(FloatCurve=DataInterface() {propertyValues {Curve {keys [<at 0 15>,<at 1 15>]}}})
    Index 0 Param 2=&FloatFromCurve(FloatCurve=DataInterface() {propertyValues {Curve {keys [<at 0 45>,<at 0.25 50>]}}})
    Index 0 Param 3=&FloatFromCurve(FloatCurve=DataInterface() {propertyValues {Curve {keys [<at 0 15>,<at 1 15>]}}})
    Index 0 Param 4=&FloatFromCurve(FloatCurve=DataInterface() {propertyValues {Curve {keys [<at 0.5 0.35>,<at 1 0>]}}})
    Module.Index0Float0Write=true
    Module.Index0Float1Write=true
    Module.Index0Float2Write=true
    Module.Index0Float3Write=true
   ) {module=DynamicMaterialParameters001}
   &ScaleColor(
    Scale Alpha=&MergeHelper_FloatSwitch(
     0=&FloatFromCurve(FloatCurve=DataInterface() {propertyValues {Curve {keys [<at 0 1 RCIM_Cubic>,<at 1 0>]}}}, Scale Curve=$System.TODScalar)
     1=&FloatFromCurve(
      FloatCurve=DataInterface() {propertyValues {Curve {keys [{arriveTangent=-2.3675, interpMode=RCIM_Cubic, leaveTangent=-2.3675, tangentMode=RCTM_User, time=0, value=0.5}, <at 1 0.25>]}}}
      Scale Curve=$System.TODScalar
     )
     2=&FloatFromCurve(
      FloatCurve=DataInterface() {propertyValues {Curve {keys [
          {arriveTangent=1.0782, time=0, value=0}
          {arriveTangent=5.5856, interpMode=RCIM_Cubic, leaveTangent=5.5856, time=0.0736, value=1.0782}
          {arriveTangent=0.3314, interpMode=RCIM_Cubic, leaveTangent=0.3314, time=0.2231, value=1.2462}
          {arriveTangent=-3.6594, interpMode=RCIM_Cubic, leaveTangent=-3.6594, time=0.2879, value=1.1492}
          {arriveTangent=-1.6745, interpMode=RCIM_Cubic, leaveTangent=-1.6745, tangentMode=RCTM_User, time=0.4685, value=0.3482}
          {arriveTangent=-0.6551, interpMode=RCIM_Cubic, leaveTangent=-0.6551, time=0.7813, value=0.1498}
          <at 1 0>
         ]}}}
      Scale Curve=$System.TODScalar
     )
     3=?
     Index=$Particles.VisibilityTag
     Num=::NewEnumerator1
    )
    Scale RGB=&MergeHelper_VectorSwitch(
     0=&VectorFromCurve(VectorCurve=DataInterface() {propertyValues {XCurve {keys [<at 0 1>,<at 1 0.7592>]}, YCurve {keys [<at 0 0.0384>,<at 1 0>]}, ZCurve {keys [<at 0 0.8441>,<at 1 1>]}}})
     1=&VariantVectorSelector(
      Original Vector:Vector3f=&VectorFromCurve(VectorCurve=DataInterface() {propertyValues {
         XCurve {keys [
           <at 0.25 0.8461 RCIM_Cubic>
           {arriveTangent=-0.4521, interpMode=RCIM_Cubic, leaveTangent=-0.4521, time=0.5, value=0.3291}
           {arriveTangent=-0.6583, interpMode=RCIM_Cubic, leaveTangent=-0.6583, time=0.7, value=0.6427}
           <at 1 0 RCIM_Cubic>
          ]}
         YCurve {keys [
           <at 0.25 0.1038 RCIM_Cubic>
           {arriveTangent=0.1823, interpMode=RCIM_Cubic, leaveTangent=0.1823, time=0.5, value=0.0779}
           {arriveTangent=-0.1559, interpMode=RCIM_Cubic, leaveTangent=-0.1559, time=0.7, value=0.1858}
           <at 1 0 RCIM_Cubic>
          ]}
         ZCurve {keys [
           <at 0.25 1 RCIM_Cubic>
           <at 0.5 0.3906 RCIM_Cubic>
           {arriveTangent=-0.7812, interpMode=RCIM_Cubic, leaveTangent=-0.7812, time=0.7, value=1}
           <at 1 0 RCIM_Cubic>
          ]}
        }})
      Variant 1 Vector:Vector3f=&VectorFromCurve(VectorCurve=DataInterface() {propertyValues {
         XCurve {keys [
           <at 0.25 0 RCIM_Cubic>
           <at 0.5 0.2917 RCIM_Cubic>
           {arriveTangent=-0.5833, interpMode=RCIM_Cubic, leaveTangent=-0.5833, time=0.7, value=0}
           <at 1 0 RCIM_Cubic>
          ]}
         YCurve {keys [
           <at 0.25 0.3401 RCIM_Cubic>
           {arriveTangent=-0.248, interpMode=RCIM_Cubic, leaveTangent=-0.248, time=0.5, value=0.5325}
           {arriveTangent=-1.0651, interpMode=RCIM_Cubic, leaveTangent=-1.0651, time=0.7, value=0.2285}
           <at 1 0 RCIM_Cubic>
          ]}
         ZCurve {keys [
           <at 0.25 1 RCIM_Cubic>
           {arriveTangent=-0.787, interpMode=RCIM_Cubic, leaveTangent=-0.787, time=0.5, value=1}
           {arriveTangent=-2, interpMode=RCIM_Cubic, leaveTangent=-2, time=0.7, value=0.6458}
           <at 1 0 RCIM_Cubic>
          ]}
        }})
      Variant 2 Vector:Vector3f=?
      Variant 3 Vector:Vector3f=?
      Variant Amount:NiagaraInt32=0
     )
     2=&VariantVectorSelector(
      Original Vector:Vector3f=&VectorFromCurve(VectorCurve=DataInterface() {propertyValues {
         XCurve {keys [{interpMode=RCIM_Cubic, tangentMode=RCTM_User, time=0, value=1}, <at 0.5 0.8 RCIM_Cubic>]}
         YCurve {keys [{arriveTangent=-546.5645, interpMode=RCIM_Cubic, leaveTangent=-546.5645, tangentMode=RCTM_User, time=0, value=250}, <at 0.5 100 RCIM_Cubic>]}
         ZCurve {keys [{interpMode=RCIM_Cubic, tangentMode=RCTM_User, time=0, value=0}, <at 0.5 0 RCIM_Cubic>]}
        }})
      Variant 1 Vector:Vector3f=&VectorFromCurve(VectorCurve=DataInterface() {propertyValues {
         XCurve {keys [{interpMode=RCIM_Cubic, tangentMode=RCTM_User, time=0, value=0}, <at 0.5 0.8 RCIM_Cubic>]}
         YCurve {keys [{arriveTangent=-546.5645, interpMode=RCIM_Cubic, leaveTangent=-546.5645, tangentMode=RCTM_User, time=0, value=85}, <at 0.5 0 RCIM_Cubic>]}
         ZCurve {keys [{arriveTangent=-989.2326, interpMode=RCIM_Cubic, leaveTangent=-989.2326, tangentMode=RCTM_User, time=0, value=250}, <at 0.5 0 RCIM_Cubic>]}
        }})
      Variant 2 Vector:Vector3f=?
      Variant 3 Vector:Vector3f=?
      Variant Amount:NiagaraInt32=0
     )
     3=?
     Index=$Particles.VisibilityTag
     Num=::NewEnumerator1
    )
    ScaleA=true
    ScaleRGB=true
   )
   &CameraOffset(Camera Offset Amount=&MergeHelper_FloatSwitch(0=0, 1=25, 2=15, 3=?, Index=$Particles.VisibilityTag, Num=::NewEnumerator1))
   &SolveForcesAndVelocity(Clamp Velocity=false, Limit Acceleration=false)
  ]
  rendererClasses [&NiagaraMeshRendererProperties]
  renderers [
   {rendererClass=&NiagaraMeshRendererProperties, rendererIndex=0}
   {rendererClass=&NiagaraMeshRendererProperties, rendererIndex=1}
   {rendererClass=&NiagaraMeshRendererProperties, rendererIndex=2}
  ]
 }
}
summary {
 systemName=NS_Character_EvilCowboy_Stage03_OnBodyTransition
 emitters [
  {emitterName=SpiralSmoke, rendererClasses [&NiagaraSpriteRendererProperties]}
  {emitterName=SpiralSmoke_Packed, rendererClasses [&NiagaraMeshRendererProperties]}
  {emitterName=ElectricArcs, rendererClasses [&NiagaraSpriteRendererProperties]}
  {
   emitterName=LightRay_Packed
   rendererClasses [
    &NiagaraSpriteRendererProperties
    &NiagaraLightRendererProperties
    &NiagaraMeshRendererProperties
   ]
  }
  {emitterName=Decal, rendererClasses [&NiagaraSpriteRendererProperties]}
 ]
 userVariables [User.Variant {type=NiagaraFloat, defaultValue=0, description=""}]
}
